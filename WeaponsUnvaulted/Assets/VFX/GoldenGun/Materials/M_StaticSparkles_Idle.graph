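# Unlit translucent material for Golden Gun idle sparkles, using a cross-flare texture and a warm gold color.
# The Sprite function billboards each quad, and a camera offset is added to its world position offset.
# A looping age from time plus vertex-color green staggers the sparkles.
# Sprite scale pulses around a peak age, masked by the dot product of normalized world position and the atmospheric light vector.
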Material /WeaponsUnvaulted/Assets/VFX/GoldenGun/Materials/M_StaticSparkles_Idle
graph {
 properties {bAutomaticallySetUsageInEditor=false, blendMode=BLEND_Translucent, shadingModel=MSM_Unlit, twoSided=true, usedShadingModels=Unlit}
 nodes [
  n1 = ScalarParameter() {defaultValue=10}
  n2 = Normalize(VectorInput=WorldPosition().XYZ)
  n3 = AtmosphericLightVector()
  n4 = Multiply(
   A=SphereMask(
    A=NamedRerouteUsage()
    B=ScalarParameter() {defaultValue=0.5, parameterName=Curve Peak Start}
    Radius=ScalarParameter() {defaultValue=0.25, parameterName=Param_3}
   ) {hardnessPercent=0}
   B=Saturate(None=DotProduct(A=@n2, B=@n3))
  )
  Add(
   A=MaterialFunctionCall!(
    Center Location=WorldPosition().XYZ
    0-1 UVs=TextureCoordinate()
    XY scale=LinearInterpolate(B=@n1, Alpha=@n4)
    OverridePivot
    "Normals (Optional)"
   ) {materialFunction=&Sprite}.World position offset
   B=MaterialFunctionCall(Offset Amount=ScalarParameter() {defaultValue=5, parameterName=Param_1}) {materialFunction=/Engine/Functions/Engine_MaterialFunctions02/WorldPositionOffset/CameraOffset}.Camera Offset
  )
  n5 = ScalarParameter() {defaultValue=1, parameterName=Param_2}
  n6 = Fmod(A=Add(A=VertexColor().G, B=Time()), B=@n5)
  NamedRerouteDeclaration(None=Divide(A=@n6, B=@n5)) {name=Normalized Age, nodeColor=<rgba 0.5 0 1 1>}
  NamedRerouteDeclaration(None=@n6) {name=Age, nodeColor=<rgba 0 1 0.1797 1>}
  TextureSample() {samplerType=SAMPLERTYPE_Grayscale, texture=/Game/Effects/Fort_Effects/Textures/Flares/T_Cross_Flare}
  VectorParameter() {defaultValue=<rgba 2 1.3157 0.2934 1>}
 ]
}
