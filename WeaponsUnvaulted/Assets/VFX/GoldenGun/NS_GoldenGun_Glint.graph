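# Niagara lens-flare glint for the Golden Gun, drawn as long-lived burst-spawned sprites in twin GPU and CPU emitters.
# Both derive a camera-vector value that drives sprite size, a secondary flare scale, alpha and alignment, with scalability distance also shaping size and fade.
# Scratch modules supply visibility to alpha: occlusion sampling on GPU, an opponent-visibility module on CPU.
# Alpha also scales by time of day.
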NiagaraSystem /WeaponsUnvaulted/Assets/VFX/GoldenGun/NS_GoldenGun_Glint
dependencies {
 usedDataInterfaces [&NiagaraDataInterfaceCurve, &NiagaraDataInterfaceCamera, /Script/Niagara.NiagaraDataInterfaceOcclusion]
 usedDynamicInputs []
 usedModules [&SystemState]
}
emitters {
 GlintLensFlarePlus {
  simTarget=GPUComputeSim
  emitterUpdateScript [
   &EmitterState(Life Cycle Mode=::NewEnumerator1, Loop Duration=1, UseLoopDelay=false)
   &SpawnBurst_Instantaneous(Use Spawn Probability=false, UseLoopCountLimit=false)
   &SpawnBurst_Instantaneous(Spawn Count=2, Spawn Time=1.2531, Use Spawn Probability=false, UseLoopCountLimit=false) {enabled=false, module=SpawnBurst_Instantaneous001}
  ]
  particleSpawnScript [
   &PartitionParticles()
   &InitializeParticle(
    Color=<rgba 0.6273 0.8504 1.3 1>
    Sprite Size Mode=::NewEnumerator3
    Uniform Sprite Size=250
    UsePositionOffset=false
    Write Material Random=true
   )
   &CameraOffset(Camera Offset Amount=15) {enabled=false}
  ]
  particleUpdateScript [
   /SetVariables(Particles.DistanceSizeScalar:NiagaraFloat=&FloatFromCurve(
     FloatCurve=DataInterface() {propertyValues {Curve {keys [
         {interpMode=RCIM_Cubic, tangentMode=RCTM_User, time=0, value=0.1}
         {arriveTangent=9.568, interpMode=RCIM_Cubic, leaveTangent=9.568, tangentMode=RCTM_User, time=0.05, value=0.8}
         {
          arriveTangent=1.4123
          arriveTangentWeight=0.0167
          interpMode=RCIM_Cubic
          leaveTangent=1.4123
          leaveTangentWeight=0.5781
          tangentMode=RCTM_User
          time=0.4
          value=2
         }
         {arriveTangentWeight=0.3181, interpMode=RCIM_Cubic, tangentMode=RCTM_User, time=1, value=2.5}
        ]}}}
     CurveIndex=$Emitter.Scalability.DistanceFraction
    )) {isSetParameters=true, module=SetVariables_#1}
   &ParticleState(Kill Particles When Lifetime Has Elapsed=false, Let Infinitely Lived Particles Die When Emitter Deactivates=true)
   &self:MaskInScreenSpace_02() {enabled=false, module=MaskInScreenSpace_02}
   &self:ScratchModule_01(OcclusionSize=20) {module=ScratchModule_01}
   /SetVariables(Particles.CameraScale:NiagaraFloat=&Multiply_Float(
     A=1
     B=&MakeFloatFromVector(
      Vector=&ReturnCameraVector(Camera Vectors=::NewEnumerator4, Coordinate Space=::Simulation, Origin Position=$Particles.Position)
      Channel=::NewEnumerator0
     )
    )) {isSetParameters=true, module=SetVariables_#2}
   &ScaleSpriteSize(
    Scale Sprite Size Mode=::NewEnumerator0
    Uniform Scale Factor=&Multiply_Float(
     A=&Lerp_Float(
      A=0.15
      B=1.5
      Alpha=&FloatFromCurve(
       FloatCurve=DataInterface() {propertyValues {Curve {keys [
           {interpMode=RCIM_Cubic, tangentMode=RCTM_User, time=0.7, value=0}
           {arriveTangent=2.314, interpMode=RCIM_Cubic, leaveTangent=2.314, tangentMode=RCTM_User, time=0.9, value=0.15}
           {arriveTangent=11.09, interpMode=RCIM_Cubic, leaveTangent=11.09, tangentMode=RCTM_User, time=1, value=1}
          ]}}}
       CurveIndex=$Particles.CameraScale
      )
     )
     B=$Particles.DistanceSizeScalar
    )
   )
   &ScaleSpriteSize(
    Initial Sprite Size:Vector2f=$Particles.SpriteSize
    Scale Sprite Size Mode=::NewEnumerator0
    Uniform Scale Factor=&Lerp_Float(A=0, B=1, Alpha=$Particles.Visibility)
   ) {enabled=false, module=ScaleSpriteSize001}
   /SetVariables(Particles.FlareScaleAlt:Vector2f=&Multiply_Vector2D_ByFloat(
     Float=&Lerp_Float(
      A=0.15
      B=1.5
      Alpha=&FloatFromCurve(
       FloatCurve=DataInterface() {propertyValues {Curve {keys [
           {interpMode=RCIM_Cubic, tangentMode=RCTM_User, time=0.7, value=0}
           {arriveTangent=3.3224, interpMode=RCIM_Cubic, leaveTangent=3.3224, tangentMode=RCTM_User, time=0.9, value=0.35}
           {arriveTangent=11.09, interpMode=RCIM_Cubic, leaveTangent=11.09, tangentMode=RCTM_User, time=1, value=1}
          ]}}}
       CurveIndex=$Particles.CameraScale
      )
     )
     Vector2D=&Multiply_Vector2D_ByFloat(Float=&Multiply_Float(A=0.6, B=$Particles.DistanceSizeScalar), Vector2D=<500 250>)
    )) {isSetParameters=true, module=SetVariables_#3}
   &ScaleColor(Scale Alpha=$Particles.Visibility, ScaleA=true, ScaleRGB=false) {module=ScaleColor002}
   &ScaleColor(
    Scale Alpha=&FloatFromCurve(
     FloatCurve=DataInterface() {propertyValues {Curve {keys [
         {tangentMode=RCTM_User, time=0, value=0}
         {arriveTangent=0.4636, interpMode=RCIM_Cubic, leaveTangent=0.4636, tangentMode=RCTM_User, time=0.8, value=0}
         {arriveTangent=0.1158, interpMode=RCIM_Cubic, leaveTangent=0.1158, tangentMode=RCTM_User, time=1, value=1}
        ]}}}
     CurveIndex=$Particles.CameraScale
     Scale Curve=&FloatFromCurve(
      FloatCurve=DataInterface() {propertyValues {Curve {keys [
          {interpMode=RCIM_Cubic, tangentMode=RCTM_User, time=0, value=0.25}
          {interpMode=RCIM_Cubic, tangentMode=RCTM_User, time=0.1, value=1}
          {arriveTangentWeight=1.0285, interpMode=RCIM_Cubic, tangentMode=RCTM_User, time=1, value=1}
         ]}}}
      CurveIndex=$Emitter.Scalability.DistanceFraction
     )
    )
    ScaleA=true
    ScaleRGB=false
   )
   &SpriteFacingAndAlignment(
    Sprite Alignment=&Lerp_Vector(
     A=&Lerp_Vector(
      A=<-1 0 1>
      B=<1 0 1>
      Alpha=&FloatFromCurve(
       FloatCurve=DataInterface() {propertyValues {Curve {keys [
           {interpMode=RCIM_Cubic, tangentMode=RCTM_User, time=0.3, value=0}
           {arriveTangentWeight=0.7423, interpMode=RCIM_Cubic, tangentMode=RCTM_User, time=1, value=1}
          ]}}}
       CurveIndex=$Particles.CameraScale
      )
     )
     B=<0 0 1>
     Alpha=&FloatFromCurve(
      FloatCurve=DataInterface() {propertyValues {Curve {keys [
          {interpMode=RCIM_Cubic, tangentMode=RCTM_User, time=0, value=0}
          {arriveTangent=0.1269, interpMode=RCIM_Cubic, leaveTangent=0.1269, tangentMode=RCTM_User, time=0.9, value=1}
          {arriveTangentWeight=0.7423, interpMode=RCIM_Cubic, tangentMode=RCTM_User, time=1, value=1}
         ]}}}
      CurveIndex=$Particles.CameraScale
     )
    )
    WriteAlignment=true
    WriteFacing=true
   ) {module=SpriteFacingAndAlignment001}
   &ScaleColor(
    Scale Alpha=&FNTimeOfDayScalar(Night Scale=0.25, Morning Scale=0.5, Evening Scale=0.5)
    ScaleA=true
    ScaleRGB=false
   ) {module=ScaleColor001}
  ]
  rendererClasses [&NiagaraSpriteRendererProperties]
  renderers [
   {rendererClass=&NiagaraSpriteRendererProperties}
   {rendererClass=&NiagaraSpriteRendererProperties, rendererIndex=1}
  ]
 }
 GlintLensFlare_CPU {
  emitterUpdateScript [
   &EmitterState(Life Cycle Mode=::NewEnumerator1, Loop Duration=1, UseLoopDelay=false)
   &SpawnBurst_Instantaneous(Use Spawn Probability=false, UseLoopCountLimit=false)
  ]
  particleSpawnScript [
   &PartitionParticles()
   &InitializeParticle(
    Color=<rgba 0.6273 0.8504 1.3 1>
    Sprite Size Mode=::NewEnumerator3
    Uniform Sprite Size=250
    UsePositionOffset=false
    Write Material Random=true
   )
  ]
  particleUpdateScript [
   /SetVariables(Particles.DistanceSizeScalar:NiagaraFloat=&FloatFromCurve(
     FloatCurve=DataInterface() {propertyValues {Curve {keys [
         {interpMode=RCIM_Cubic, tangentMode=RCTM_User, time=0, value=0.1}
         {arriveTangent=9.568, interpMode=RCIM_Cubic, leaveTangent=9.568, tangentMode=RCTM_User, time=0.05, value=0.8}
         {
          arriveTangent=1.4123
          arriveTangentWeight=0.0167
          interpMode=RCIM_Cubic
          leaveTangent=1.4123
          leaveTangentWeight=0.5781
          tangentMode=RCTM_User
          time=0.4
          value=2
         }
         {arriveTangentWeight=0.3181, interpMode=RCIM_Cubic, tangentMode=RCTM_User, time=1, value=2.5}
        ]}}}
     CurveIndex=$Emitter.Scalability.DistanceFraction
    )) {isSetParameters=true, module=SetVariables_#1}
   &ParticleState(Kill Particles When Lifetime Has Elapsed=false, Let Infinitely Lived Particles Die When Emitter Deactivates=true)
   &self:Opponent_Visibility() {module=Opponent_Visibility}
   /SetVariables(Particles.CameraScale:NiagaraFloat=&Multiply_Float(
     A=1
     B=&MakeFloatFromVector(
      Vector=&ReturnCameraVector(Camera Vectors=::NewEnumerator4, Coordinate Space=::Simulation, Origin Position=$Particles.Position)
      Channel=::NewEnumerator0
     )
    )) {isSetParameters=true, module=SetVariables_#2}
   &ScaleSpriteSize(
    Scale Sprite Size Mode=::NewEnumerator0
    Uniform Scale Factor=&Multiply_Float(
     A=&Lerp_Float(
      A=0.15
      B=1.5
      Alpha=&FloatFromCurve(
       FloatCurve=DataInterface() {propertyValues {Curve {keys [
           {interpMode=RCIM_Cubic, tangentMode=RCTM_User, time=0.7, value=0}
           {arriveTangent=2.314, interpMode=RCIM_Cubic, leaveTangent=2.314, tangentMode=RCTM_User, time=0.9, value=0.15}
           {arriveTangent=11.09, interpMode=RCIM_Cubic, leaveTangent=11.09, tangentMode=RCTM_User, time=1, value=1}
          ]}}}
       CurveIndex=$Particles.CameraScale
      )
     )
     B=$Particles.DistanceSizeScalar
    )
   )
   /SetVariables(Particles.FlareScaleAlt:Vector2f=&Multiply_Vector2D_ByFloat(
     Float=&Lerp_Float(
      A=0.15
      B=1.5
      Alpha=&FloatFromCurve(
       FloatCurve=DataInterface() {propertyValues {Curve {keys [
           {interpMode=RCIM_Cubic, tangentMode=RCTM_User, time=0.7, value=0}
           {arriveTangent=3.3224, interpMode=RCIM_Cubic, leaveTangent=3.3224, tangentMode=RCTM_User, time=0.9, value=0.35}
           {arriveTangent=11.09, interpMode=RCIM_Cubic, leaveTangent=11.09, tangentMode=RCTM_User, time=1, value=1}
          ]}}}
       CurveIndex=$Particles.CameraScale
      )
     )
     Vector2D=&Multiply_Vector2D_ByFloat(Float=&Multiply_Float(A=0.6, B=$Particles.DistanceSizeScalar), Vector2D=<500 250>)
    )) {isSetParameters=true, module=SetVariables_#3}
   &ScaleColor(Scale Alpha=$Particles.Visibility, ScaleA=true, ScaleRGB=false) {module=ScaleColor002}
   &ScaleColor(
    Scale Alpha=&FloatFromCurve(
     FloatCurve=DataInterface() {propertyValues {Curve {keys [
         {tangentMode=RCTM_User, time=0, value=0}
         {arriveTangent=0.4636, interpMode=RCIM_Cubic, leaveTangent=0.4636, tangentMode=RCTM_User, time=0.8, value=0}
         {arriveTangent=0.1158, interpMode=RCIM_Cubic, leaveTangent=0.1158, tangentMode=RCTM_User, time=1, value=1}
        ]}}}
     CurveIndex=$Particles.CameraScale
     Scale Curve=&FloatFromCurve(
      FloatCurve=DataInterface() {propertyValues {Curve {keys [
          {interpMode=RCIM_Cubic, tangentMode=RCTM_User, time=0, value=0.25}
          {interpMode=RCIM_Cubic, tangentMode=RCTM_User, time=0.1, value=1}
          {arriveTangentWeight=1.0285, interpMode=RCIM_Cubic, tangentMode=RCTM_User, time=1, value=1}
         ]}}}
      CurveIndex=$Emitter.Scalability.DistanceFraction
     )
    )
    ScaleA=true
    ScaleRGB=false
   )
   &SpriteFacingAndAlignment(
    Sprite Alignment=&Lerp_Vector(
     A=&Lerp_Vector(
      A=<-1 0 1>
      B=<1 0 1>
      Alpha=&FloatFromCurve(
       FloatCurve=DataInterface() {propertyValues {Curve {keys [
           {interpMode=RCIM_Cubic, tangentMode=RCTM_User, time=0.3, value=0}
           {arriveTangentWeight=0.7423, interpMode=RCIM_Cubic, tangentMode=RCTM_User, time=1, value=1}
          ]}}}
       CurveIndex=$Particles.CameraScale
      )
     )
     B=<0 0 1>
     Alpha=&FloatFromCurve(
      FloatCurve=DataInterface() {propertyValues {Curve {keys [
          {interpMode=RCIM_Cubic, tangentMode=RCTM_User, time=0, value=0}
          {arriveTangent=0.1269, interpMode=RCIM_Cubic, leaveTangent=0.1269, tangentMode=RCTM_User, time=0.9, value=1}
          {arriveTangentWeight=0.7423, interpMode=RCIM_Cubic, tangentMode=RCTM_User, time=1, value=1}
         ]}}}
      CurveIndex=$Particles.CameraScale
     )
    )
    WriteAlignment=true
    WriteFacing=true
   ) {module=SpriteFacingAndAlignment001}
   &ScaleColor(
    Scale Alpha=&FNTimeOfDayScalar(Night Scale=0.25, Morning Scale=0.5, Evening Scale=0.5)
    ScaleA=true
    ScaleRGB=false
   ) {module=ScaleColor001}
  ]
  rendererClasses [&NiagaraSpriteRendererProperties]
  renderers [
   {rendererClass=&NiagaraSpriteRendererProperties}
   {rendererClass=&NiagaraSpriteRendererProperties, rendererIndex=1}
  ]
 }
}
summary {
 systemName=NS_GoldenGun_Glint
 emitters [
  {emitterName=GlintLensFlarePlus, simTarget=GPUComputeSim, rendererClasses [&NiagaraSpriteRendererProperties]}
  {emitterName=GlintLensFlare_CPU, rendererClasses [&NiagaraSpriteRendererProperties]}
 ]
 userVariables [
  User.GlobalOpacity {type=NiagaraFloat, defaultValue=1, description=""}
  User.GlobalScale {type=NiagaraFloat, defaultValue=1, description=""}
  User.SkeletalMesh {
   type=NiagaraDataInterfaceSkeletalMesh
   defaultValue {
    struct=NiagaraExt_VariableValue_DataInterface
    value {dataInterface=&self:NiagaraDataInterfaceSkeletalMesh_1}
   }
  }
 ]
}
