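# Niagara handler system for Golden Gun bullet impacts, spawning effects from impact data read through a Niagara Data Channel.
# Emitters in the visible portion include fading crack and scorch decals, surface-tinted sprite puffs (Flames), a gold mesh, sprite and light flash, and physical-material mesh debris.
# Each spawns from the shared NDC reader, taking position, normal, surface and ammo scale from the channel.
# Physical surface selects color, lifetime, motion and debris type, while ImpactDelay sets initial particle age.
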NiagaraSystem /WeaponsUnvaulted/Assets/VFX/GoldenGun/NS_NDC_Handler_Impacts_GoldenGun
dependencies {
 usedDataInterfaces [
  &NiagaraDataInterfaceArrayFloat
  &NiagaraDataInterfaceCurve
  &NiagaraDataInterfaceColorCurve
  &NiagaraDataInterfaceVectorCurve
  &NiagaraDataInterfaceArrayColor
  &NiagaraDataInterfaceVector2DCurve
  /Script/Niagara.NiagaraDataInterfaceDataChannelRead
 ]
 usedDynamicInputs [
  &SelectFloatFromArray
  &MakeCustomLinearColorFromBool
  &Boolean_OrOperation
  &Vector2DFromFloat
 ]
 usedModules [
  &self:SystemSpawnScript.SystemScriptSource.SystemScriptGraph.NiagaraNodeAssignment_2.SetVariables_#1
  &SystemState
  &self:SystemSpawnScript.SystemScriptSource.SystemScriptGraph.NiagaraNodeAssignment_1.SetVariables_#2
  /Niagara/Modules/System/CompleteIfUnused
  &self:SystemSpawnScript.SystemScriptSource.SystemScriptGraph.NiagaraNodeAssignment_0.SetVariables_#3
 ]
}
emitters {
 DecalRenderer {
  emitterUpdateScript [
   &EmitterState()
   /SetVariables(
    Emitter.OpacityMultCrack:NiagaraFloat=&RandomRangeFloat(Minimum=1, Maximum=1.5)
    Emitter.RandRotCracks:NiagaraFloat=&RandomRangeFloat(Minimum=0, Maximum=1)
   ) {isSetParameters=true}
   /Game/Effects/Niagara/DataChannels/Impacts/Spawn_NDCFortImpacts(Reader:NiagaraDataInterfaceDataChannelRead=$System.NDCReader)
  ]
  particleSpawnScript [
   /SetVariables() {isSetParameters=true}
   /Game/Effects/Niagara/DataChannels/Impacts/InitializeParticleFromNDC_FortImpact(
    AmmoType Scale=1
    FallbackImpactScale=50
    Scale Array=DataInterface() {propertyValues {GpuSyncMode=SyncCpuToGpu, MaxElements=0, FloatData [24,35,50,20,80]}}
   )
   &InitializeParticle(
    Lifetime=&Lerp_Float(
     A=1.6
     B=2
     Alpha=&FloatFromCurve(
      FloatCurve=DataInterface() {propertyValues {Curve {keys [
          {interpMode=RCIM_Cubic, tangentMode=RCTM_User, time=0, value=0}
          {arriveTangent=3.443, interpMode=RCIM_Cubic, leaveTangent=3.443, tangentMode=RCTM_User, time=1, value=1}
         ]}}}
      CurveIndex=&RandomRangeFloat(Minimum=0, Maximum=1)
     )
    )
    Position=$Particles.Position
    Position Mode=::NewEnumerator1
    UsePositionOffset=true
    Write Material Random=true
   )
   /Niagara/Modules/Spawn/Decal/InitialDecalOrientation(
    Orientation Vector:Vector3f=$Particles.ImpactSurfaceNormal
    Rotation:Vector3f=&RandomRangeVector(Minimum=<0 0 0>, Maximum=<0 0 1>)
    Use Rotation Vector=true
   ) {module=InitialDecalOrientation001}
  ]
  particleUpdateScript [
   &ParticleState()
   /SetVariables(
    Particles.DecalFadeCrack:NiagaraFloat=&Lerp_Float(
     A=0
     B=1
     Alpha=&FloatFromCurve(FloatCurve=DataInterface() {propertyValues {Curve {keys [
          {arriveTangent=41.722, time=0, value=0}
          {interpMode=RCIM_Cubic, tangentMode=RCTM_User, time=0.02, value=1}
          {arriveTangent=-1.7378, interpMode=RCIM_Cubic, leaveTangent=-1.7378, tangentMode=RCTM_User, time=0.7, value=0.7}
          {arriveTangent=-0.5041, interpMode=RCIM_Cubic, leaveTangent=-0.5041, tangentMode=RCTM_User, time=1, value=0}
         ]}}})
    )
    Particles.DecalFadeScorch:NiagaraFloat=&Lerp_Float(
     A=0
     B=1
     Alpha=&FloatFromCurve(FloatCurve=DataInterface() {propertyValues {Curve {keys [
          {arriveTangent=41.722, time=0, value=0.25}
          {interpMode=RCIM_Cubic, tangentMode=RCTM_User, time=0.02, value=1}
          {arriveTangent=-1.795, interpMode=RCIM_Cubic, leaveTangent=-1.795, tangentMode=RCTM_User, time=0.6, value=0.6}
          {arriveTangent=-3.611, arriveTangentWeight=0.112, interpMode=RCIM_Cubic, leaveTangent=-3.611, tangentMode=RCTM_User, time=1, value=0}
         ]}}})
    )
   ) {isSetParameters=true}
   /SetVariables(
    Particles.DecalSizeCrack:Vector3f=&Lerp_Vector(
     A=&Lerp_Vector(A=<100 62 62>, B=<100 100 100>, Alpha=1)
     B=&Lerp_Vector(A=<100 128 128>, B=<100 200 200>, Alpha=1)
     Alpha=&FloatFromCurve(FloatCurve=DataInterface() {propertyValues {Curve {keys [
          {leaveTangentWeight=0.6956, time=0, value=0}
          {arriveTangentWeight=0.0167, leaveTangentWeight=0.25, time=0.01, value=1}
          {arriveTangentWeight=0.4545, interpMode=RCIM_Cubic, leaveTangentWeight=0.2005, tangentMode=RCTM_User, time=0.6, value=1}
          {arriveTangent=-1.586, interpMode=RCIM_Cubic, leaveTangent=-1.586, tangentMode=RCTM_User, time=1, value=0.8}
         ]}}})
    )
    Particles.DecalSizeEmissive:Vector3f=&Multiply_VectorByFloat(
     Vector=&Multiply_VectorByFloat(Vector=<15 250 250>, Float=0.16)
     Float=&RandomRangeFloat(Minimum=0.7, Maximum=1)
    )
   ) {isSetParameters=true}
  ]
  rendererClasses [&NiagaraDecalRendererProperties]
  renderers [{rendererClass=&NiagaraDecalRendererProperties, rendererIndex=0}]
 }
 Flames {
  emitterUpdateScript [
   &EmitterState()
   &Spawn_NDCFortImpacts(
    Max Spawn Count=5
    Min Spawn Count=4
    Reader:NiagaraDataInterfaceDataChannelRead=$System.NDCReader
    Simple Max Spawn Count=2
    Simple Min Spawn Count=2
    Simple Spawn Override=2
    Spawn Count Mode=2
   )
   &Spawn_NDCFortImpacts(Physical Material=SurfaceType3, Reader:NiagaraDataInterfaceDataChannelRead=$System.NDCReader, Surface Type=1) {enabled=false, module=Spawn_NDCFortImpacts001}
  ]
  particleSpawnScript [
   /SetVariables(Particles.Age:NiagaraFloat=$System.ImpactDelay) {isSetParameters=true}
   &InitializeParticleFromNDC_FortImpact(
    AmmoType Scale=1
    FallbackPosition:NiagaraPosition=<0 0 1>
    Scale Array=DataInterface() {propertyValues {GpuSyncMode=SyncCpuToGpu, MaxElements=4, FloatData [1.6,2,2.4,1.4,2]}}
   )
   /SetVariables(Particles.VisibilityTag=/Game/Effects/Niagara/DataChannels/SelectIntBySurface(
     Default:NiagaraInt32=0
     Surface:EPhysicalSurface=$Particles.PhysicalSurface
     Num Surfaces:NiagaraInt32=4
     Surface 0:EPhysicalSurface=SurfaceType23
     Value For Surface 0:NiagaraInt32=1
     Surface 1:EPhysicalSurface=SurfaceType24
     Value For Surface 1:NiagaraInt32=1
     Surface 2:EPhysicalSurface=SurfaceType26
     Value For Surface 2:NiagaraInt32=2
     Surface 3:EPhysicalSurface=SurfaceType3
     Value For Surface 3:NiagaraInt32=3
    )) {isSetParameters=true}
   /SetVariables(
    Particles.LifetimeMax:NiagaraFloat=/Game/Effects/Niagara/DataChannels/SelectFloatBySurface(
     Default:NiagaraFloat=1
     Surface:EPhysicalSurface=$Particles.PhysicalSurface
     Num Surfaces:NiagaraInt32=4
     Surface 0:EPhysicalSurface=SurfaceType23
     Value For Surface 0:NiagaraFloat=1
     Surface 1:EPhysicalSurface=SurfaceType24
     Value For Surface 1:NiagaraFloat=0.4
     Surface 2:EPhysicalSurface=SurfaceType26
     Value For Surface 2:NiagaraFloat=1.25
     Surface 3:EPhysicalSurface=SurfaceType3
     Value For Surface 3:NiagaraFloat=2
    )
    Particles.LifetimeMin:NiagaraFloat=&SelectFloatBySurface(
     Default:NiagaraFloat=0.2
     Surface:EPhysicalSurface=$Particles.PhysicalSurface
     Num Surfaces:NiagaraInt32=4
     Surface 0:EPhysicalSurface=SurfaceType23
     Value For Surface 0:NiagaraFloat=0.5
     Surface 1:EPhysicalSurface=SurfaceType24
     Value For Surface 1:NiagaraFloat=0.2
     Surface 2:EPhysicalSurface=SurfaceType26
     Value For Surface 2:NiagaraFloat=0.5
     Surface 3:EPhysicalSurface=SurfaceType3
     Value For Surface 3:NiagaraFloat=1.75
    )
   ) {isSetParameters=true}
   /SetVariables(Particles.SizeMax=85, Particles.SizeMin=35) {isSetParameters=true}
   /SetVariables(Particles.SizeCurve:NiagaraFloat=&Multiply_Float(
     A=&Lerp_Float(
      A=$Particles.SizeMin
      B=$Particles.SizeMax
      Alpha=&FloatFromCurve(
       FloatCurve=DataInterface() {propertyValues {Curve {keys [
           {interpMode=RCIM_Cubic, tangentMode=RCTM_Break, time=0, value=0}
           {arriveTangent=2.5778, interpMode=RCIM_Cubic, tangentMode=RCTM_Break, time=1, value=1}
          ]}}}
       CurveIndex=&ReturnNormalizedExecIndex()
      )
     )
     B=&SelectFloatBySurface(
      Default:NiagaraFloat=1
      Surface:EPhysicalSurface=$Particles.PhysicalSurface
      Num Surfaces:NiagaraInt32=4
      Surface 0:EPhysicalSurface=SurfaceType26
      Value For Surface 0:NiagaraFloat=1.3
      Surface 1:EPhysicalSurface=SurfaceType24
      Value For Surface 1:NiagaraFloat=0.8
      Surface 2:EPhysicalSurface=SurfaceType23
      Value For Surface 2:NiagaraFloat=1.25
      Surface 3:EPhysicalSurface=SurfaceType3
      Value For Surface 3:NiagaraFloat=1
     )
    )) {isSetParameters=true}
   &InitializeParticle(
    Color Mode=::NewEnumerator0
    Lifetime=&Multiply_Float(
     A=&Lerp_Float(A=$Particles.LifetimeMin, B=$Particles.LifetimeMax, Alpha=&ReturnNormalizedExecIndex())
     B=1
    )
    Position=$Particles.Position
    Position Mode=::NewEnumerator1
    Sprite Rotation Mode=::NewEnumerator2
    Sprite Size Mode=::NewEnumerator3
    Uniform Sprite Size=&Multiply_Float(A=$Particles.SizeCurve, B=$Particles.AmmoTypeScale)
    UsePositionOffset=false
    Write Material Random=true
   )
   /SetVariables(Particles.Color=&SelectLinearColorFromArray(
     Array Sampling Mode=::NewEnumerator1
     Direct Array Index=$Particles.VisibilityTag
     Color Selection Array=DataInterface() {propertyValues {ColorData [<rgba 0.66 0.5816 0.528 0.15>, <rgba 0.9 0.966 1 0.75>, <rgba 0.8 0.7328 0.704 0.75>, <rgba 0.25 0.25 0.25 0.5>]}}
     Random Seed=?
    )) {isSetParameters=true}
   /SetVariables(Particles.InitSpriteRotRate=&UniformRangedFloat(Minimum=-150, Maximum=150)) {isSetParameters=true}
   &AddVelocityInCone(
    Cone Axis=$Particles.ImpactSurfaceNormal
    Cone Axis Coordinate Space=::Simulation
    Module.Use Velocity Falloff On Cone Axis=false
    Velocity Strength=&Multiply_Float(A=&RandomRangeFloat(Minimum=5, Maximum=500), B=$Particles.AmmoTypeScale)
   )
   &AddVelocityInCone(
    Cone Angle=90
    Cone Axis=$Particles.ImpactSurfaceNormal
    Cone Axis Coordinate Space=::Simulation
    Module.Use Velocity Falloff On Cone Axis=false
    Velocity Strength=&Multiply_Float(
     A=&Lerp_Float(
      A=100
      B=250
      Alpha=&NormalizeFloat(Value=$Particles.SizeCurve, Min=$Particles.SizeMin, Max=$Particles.SizeMax)
     )
     B=$Particles.AmmoTypeScale
    )
   ) {enabled=false, module=AddVelocityInCone001}
   &AddVelocity(
    Rotation Coordinate Space=::World
    Velocity=<0 0 1>
    Velocity Speed Scale=&Multiply_Float(A=50, B=$Particles.AmmoTypeScale)
   ) {enabled=false, module=AddVelocity001}
   /SetVariables(Particles.Velocity=/Game/Effects/Niagara/DataChannels/SelectVectorBySurface(
     Default:Vector3f=$Particles.Velocity
     Surface:EPhysicalSurface=$Particles.PhysicalSurface
     Num Surfaces:NiagaraInt32=1
     Surface 0:EPhysicalSurface=SurfaceType3
     Value For Surface 0:Vector3f=&Multiply_VectorByFloat(Vector=$Particles.Velocity, Float=&ReturnNormalizedExecIndex(Normalized Index Scale=0.25))
     Surface 1:EPhysicalSurface=?
     Value For Surface 1:Vector3f=?
     Surface 2:EPhysicalSurface=?
     Value For Surface 2:Vector3f=?
     Surface 3:EPhysicalSurface=?
     Value For Surface 3:Vector3f=?
    )) {isSetParameters=true}
   /SetVariables(Particles.DynamicMaterialParameter_Snow:Vector4f=<1 0.6 1 1>) {isSetParameters=true}
   /SetVariables(Particles.DynamicMaterialParameter_Sand:Vector4f=&MakeVector4(X=1, Y=0.66, Z=1, W=1)) {isSetParameters=true}
   /SetVariables(Particles.DynamicMaterialParameter_Smoke:Vector4f=&MakeVector4(X=0, Y=0.25, Z=1, W=1)) {isSetParameters=true}
   /SetVariables(Particles.SubUVRandom:NiagaraInt32=&RandomRangeInt(
     Minimum=0
     Maximum=&SelectIntBySurface(
      Default:NiagaraInt32=$Particles.PhysicalSurfaceInteger
      Surface:EPhysicalSurface=SurfaceType_Default
      Num Surfaces:NiagaraInt32=1
      Surface 0:EPhysicalSurface=SurfaceType3
      Value For Surface 0:NiagaraInt32=32
      Surface 1:EPhysicalSurface=?
      Value For Surface 1:NiagaraInt32=?
      Surface 2:EPhysicalSurface=?
      Value For Surface 2:NiagaraInt32=?
      Surface 3:EPhysicalSurface=?
      Value For Surface 3:NiagaraInt32=?
     )
    )) {isSetParameters=true}
   &KillParticles(Kill Particles=&SetBoolByIntComparison(A=$Particles.PhysicalSurfaceInteger, B=8, Comparison Type=::NewEnumerator4))
  ]
  particleUpdateScript [
   &ParticleState()
   &Drag(
    Drag=&SelectFloatBySurface(
     Default:NiagaraFloat=5
     Surface:EPhysicalSurface=$Particles.PhysicalSurface
     Num Surfaces:NiagaraInt32=4
     Surface 0:EPhysicalSurface=SurfaceType23
     Value For Surface 0:NiagaraFloat=5
     Surface 1:EPhysicalSurface=SurfaceType24
     Value For Surface 1:NiagaraFloat=3
     Surface 2:EPhysicalSurface=SurfaceType26
     Value For Surface 2:NiagaraFloat=5
     Surface 3:EPhysicalSurface=SurfaceType3
     Value For Surface 3:NiagaraFloat=3
    )
    Use Linear Drag=true
    Use Rotational Drag=false
   )
   &GravityForce(Gravity=&SelectVectorBySurface(
     Default:Vector3f=<0 0 0>
     Surface:EPhysicalSurface=$Particles.PhysicalSurface
     Num Surfaces:NiagaraInt32=4
     Surface 0:EPhysicalSurface=SurfaceType23
     Value For Surface 0:Vector3f=<0 0 -980>
     Surface 1:EPhysicalSurface=SurfaceType24
     Value For Surface 1:Vector3f=<0 0 -490>
     Surface 2:EPhysicalSurface=SurfaceType26
     Value For Surface 2:Vector3f=&Multiply_VectorByFloat(Vector=<0 0 -980>, Float=1)
     Surface 3:EPhysicalSurface=SurfaceType3
     Value For Surface 3:Vector3f=&Multiply_VectorByFloat(Vector=<0 0 150>, Float=&ReturnNormalizedExecIndex())
    ))
   &SpriteRotationRate(Rotation Rate=&Multiply_Float(
     A=$Particles.InitSpriteRotRate
     B=&FloatFromCurve(
      FloatCurve=DataInterface() {propertyValues {Curve {keys [{arriveTangent=-2.0511, interpMode=RCIM_Cubic, leaveTangent=-2.0511, tangentMode=RCTM_User, time=0, value=1}, <at 1 0.1 RCIM_Cubic>]}}}
      Scale Curve=&SelectFloatBySurface(
       Default:NiagaraFloat=1
       Surface:EPhysicalSurface=$Particles.PhysicalSurface
       Num Surfaces:NiagaraInt32=3
       Surface 0:EPhysicalSurface=SurfaceType_Default
       Value For Surface 0:NiagaraFloat=1
       Surface 1:EPhysicalSurface=SurfaceType1
       Value For Surface 1:NiagaraFloat=1
       Surface 2:EPhysicalSurface=SurfaceType26
       Value For Surface 2:NiagaraFloat=1
       Surface 3:EPhysicalSurface=?
       Value For Surface 3:NiagaraFloat=?
      )
     )
    ))
   &SolveForcesAndVelocity(
    Clamp Velocity=false
    DeltaTime:NiagaraFloat=&MakeCustomFloatFromBool(Bool=&SetBoolByFloatComparison(A=$Particles.Age, B=0), True Float=$Engine.DeltaTime, False Float=0)
    Limit Acceleration=false
   )
   &SubUVAnimation(
    End Frame:NiagaraInt32=&Add_Integer(A=$Particles.SubUVRandom, B=63)
    Start Frame:NiagaraInt32=$Particles.SubUVRandom
    SubUV Loop Count:NiagaraFloat=&SelectFloatBySurface(
     Default:NiagaraFloat=0
     Surface:EPhysicalSurface=$Particles.PhysicalSurface
     Num Surfaces:NiagaraInt32=4
     Surface 0:EPhysicalSurface=SurfaceType23
     Value For Surface 0:NiagaraFloat=0.66
     Surface 1:EPhysicalSurface=SurfaceType24
     Value For Surface 1:NiagaraFloat=0.33
     Surface 2:EPhysicalSurface=SurfaceType26
     Value For Surface 2:NiagaraFloat=0.75
     Surface 3:EPhysicalSurface=SurfaceType3
     Value For Surface 3:NiagaraFloat=0.5
    )
   )
   &ScaleSpriteSize(Uniform Curve Sprite Scale=DataInterface() {propertyValues {Curve {keys [
        {arriveTangent=2.9194, interpMode=RCIM_Cubic, leaveTangent=2.9194, tangentMode=RCTM_User, time=0, value=0.25}
        {arriveTangent=1.1179, interpMode=RCIM_Cubic, leaveTangent=1.1179, tangentMode=RCTM_User, time=0.2, value=0.65}
        {arriveTangent=0.0224, interpMode=RCIM_Cubic, leaveTangent=0.0224, tangentMode=RCTM_User, time=1, value=1}
       ]}}})
   /Game/Effects/Niagara/Modules/Materials/MaterialSmoke(
    Erode:NiagaraFloat=&FloatFromCurve(FloatCurve=DataInterface() {propertyValues {
       bOptimizeLUT=false
       bUseLUT=false
       Curve {keys [
         {arriveTangent=-0.0652, interpMode=RCIM_Cubic, leaveTangent=-0.0652, tangentMode=RCTM_User, time=0.11, value=0.97}
         {arriveTangent=-5.7797, interpMode=RCIM_Cubic, leaveTangent=-5.7797, tangentMode=RCTM_User, time=0.39, value=0.66}
         {arriveTangent=-0.342, interpMode=RCIM_Cubic, leaveTangent=-0.342, tangentMode=RCTM_User, time=0.9954, value=-0.74}
        ]}
      }})
    Shading:NiagaraFloat=&FloatFromCurve(FloatCurve=DataInterface() {propertyValues {
       bOptimizeLUT=false
       bUseLUT=false
       Curve {keys [
         <at 0.07 0.5 RCIM_Cubic>
         {arriveTangent=1.2205, interpMode=RCIM_Cubic, leaveTangent=1.2205, time=0.17, value=1.02}
         {interpMode=RCIM_Cubic, tangentMode=RCTM_User, time=0.3895, value=0.89}
         <at 0.8 0 RCIM_Cubic>
        ]}
      }})
   ) {enabled=false}
   &DynamicMaterialParameters(
    Index 0 Param 1=&Add_Float(
     A=&RandomRangeFloat(Minimum=0, Maximum=1)
     B=&FloatFromCurve(
      FloatCurve=DataInterface() {propertyValues {Curve {keys [
          {arriveTangent=2.7533, interpMode=RCIM_Cubic, leaveTangent=1.8958, tangentMode=RCTM_User, time=0, value=1}
          {arriveTangent=0.8292, interpMode=RCIM_Cubic, leaveTangent=0.8292, tangentMode=RCTM_User, time=0.321, value=1.3453}
          {arriveTangent=0.0529, interpMode=RCIM_Cubic, leaveTangent=0.2021, tangentMode=RCTM_User, time=1, value=1.5}
         ]}}}
      CurveIndex=&Add_Float(A=&Multiply_Float(A=$Particles.Initial.MaterialRandom, B=0.5), B=$Particles.NormalizedAge)
     )
    )
    Index 0 Param 2=&FloatFromCurve(FloatCurve=DataInterface() {propertyValues {Curve {keys [
         {arriveTangent=0.8564, interpMode=RCIM_Cubic, leaveTangent=0.8564, tangentMode=RCTM_User, time=0.0046, value=-0.9753}
         {arriveTangent=4.2876, interpMode=RCIM_Cubic, leaveTangent=5.5268, tangentMode=RCTM_User, time=1, value=1}
        ]}}})
    Index 2 Param 1=&FloatFromCurve(FloatCurve=DataInterface() {propertyValues {Curve {keys [
         {interpMode=RCIM_Cubic, leaveTangent=4.2485, tangentMode=RCTM_Break, time=0, value=0}
         {arriveTangent=1.0548, interpMode=RCIM_Cubic, leaveTangent=1.0548, tangentMode=RCTM_User, time=0.17, value=0.5715}
         <at 1 1>
        ]}}})
    Index 2 Param 2=&RandomRangeFloat(Minimum=0, Maximum=1)
    Index 2 Param 3=&FloatFromCurve(
     FloatCurve=DataInterface() {propertyValues {Curve {keys [
         {interpMode=RCIM_Cubic, leaveTangentWeight=0.3354, tangentMode=RCTM_Break, time=0.4, value=0}
         {arriveTangent=2.7846, interpMode=RCIM_Cubic, leaveTangent=11.1233, tangentMode=RCTM_Break, time=1, value=1}
        ]}}}
     Scale Curve=7
    )
    Index 2 Param 4=&FloatFromCurve(
     FloatCurve=DataInterface() {propertyValues {Curve {keys [
         {interpMode=RCIM_Cubic, leaveTangentWeight=0.4044, tangentMode=RCTM_Break, time=0, value=0.8}
         {arriveTangent=0.4016, arriveTangentWeight=0.382, interpMode=RCIM_Cubic, tangentMode=RCTM_Break, time=1, value=1}
        ]}}}
     Scale Curve=0.5
    )
    Module.Param0WriteEnabled=true
    Module.Param0WriteEnabled2=true
    Module.Param1WriteEnabled=true
    Module.Param1WriteEnabled2=true
    Module.Param2WriteEnabled=true
    Module.Param2WriteEnabled2=true
    Module.Param3WriteEnabled=true
    Module.Param3WriteEnabled2=true
    Write Parameter Index 2=true
   )
   /SetVariables(Particles.DynamicMaterialParameter_Sand:Vector4f=&MakeVector4(
     X=&Add_Float(
      A=&RandomRangeFloat(Minimum=0, Maximum=1)
      B=&FloatFromCurve(
       FloatCurve=DataInterface() {propertyValues {Curve {keys [
           {arriveTangent=0.7934, interpMode=RCIM_Cubic, leaveTangent=0.7934, tangentMode=RCTM_User, time=0, value=1}
           {interpMode=RCIM_Cubic, leaveTangent=0.2021, time=1, value=1.5}
          ]}}}
       CurveIndex=&Add_Float(A=&Multiply_Float(A=$Particles.Initial.MaterialRandom, B=0.5), B=$Particles.NormalizedAge)
      )
     )
     Y=0
     Z=0
     W=&FloatFromCurve(FloatCurve=DataInterface() {propertyValues {Curve {keys [<at 0 1 RCIM_Cubic>,<at 1 0.25>]}}})
    )) {enabled=false, isSetParameters=true}
   /SetVariables(Particles.DynamicMaterialParameter2_Sand:Vector4f=&MakeVector4(
     X=&FloatFromCurve(FloatCurve=DataInterface() {propertyValues {Curve {keys [{tangentMode=RCTM_User, time=0, value=0}, {tangentMode=RCTM_User, time=1, value=1}]}}})
     Y=&RandomRangeFloat(Minimum=-0.1, Maximum=0.1)
     Z=&Lerp_Float(
      A=1
      B=6
      Alpha=&FloatFromCurve(FloatCurve=DataInterface() {propertyValues {Curve {keys [{interpMode=RCIM_Cubic, tangentMode=RCTM_User, time=0.33, value=0}, {interpMode=RCIM_Cubic, tangentMode=RCTM_User, time=0.66, value=1}]}}})
     )
     W=&Multiply_Float(
      A=&RandomRangeFloat(Minimum=1.5, Maximum=1.75)
      B=&Lerp_Float(
       A=1.25
       B=1
       Alpha=&FloatFromCurve(FloatCurve=DataInterface() {propertyValues {Curve {keys [{tangentMode=RCTM_User, time=0, value=0}, {tangentMode=RCTM_User, time=1, value=1}]}}})
      )
     )
    )) {enabled=false, isSetParameters=true}
   &ScaleColor(
    Scale Alpha=&FloatFromCurve(FloatCurve=DataInterface() {propertyValues {Curve {defaultValue=1, keys [{tangentMode=RCTM_User, time=0, value=1}, {tangentMode=RCTM_User, time=1, value=0}]}}})
    Scale RGB=&MakeVectorFromLinearColorRGB(Color=&MergeHelper_ColorSwitch(
      0=&ColorFromCurve(ColorCurve=DataInterface() {propertyValues {
         AlphaCurve {keys [<at 0 1>,<at 1 0>]}
         BlueCurve {keys [<at 0.0156 0.0019>,<at 0.5334 1>,<at 1 1>]}
         GreenCurve {keys [<at 0.0156 0.7006>,<at 0.5334 0.9388>,<at 1 1>]}
         RedCurve {keys [<at 0.0156 1>,<at 0.5334 0.949>,<at 1 1>]}
        }})
      1=&ColorFromCurve(ColorCurve=DataInterface() {propertyValues {
         AlphaCurve {keys [<at 0 1>,<at 1 0>]}
         BlueCurve {keys [<at 0.0156 0>,<at 0.5334 1>,<at 1 1>]}
         GreenCurve {keys [<at 0.0156 0.7>,<at 0.5334 0.9388>,<at 1 1>]}
         RedCurve {keys [<at 0.0156 1>,<at 0.5334 0.949>,<at 1 1>]}
        }})
      2=<rgba 1 1 1 1>
      3=&ColorFromCurve(ColorCurve=DataInterface() {propertyValues {
         AlphaCurve {keys [<at 0 1>,<at 1 0>]}
         BlueCurve {keys [<at 0 0>,<at 0.5334 1>,<at 1 1>]}
         GreenCurve {keys [<at 0 0.6612>,<at 0.5334 0.9388>,<at 1 1>]}
         RedCurve {keys [<at 0 1>,<at 0.5334 0.949>,<at 1 1>]}
        }})
      Index=$Particles.VisibilityTag
      Num=::NewEnumerator2
     ))
    ScaleA=true
    ScaleRGB=true
   )
   &CameraOffset(Camera Offset Amount=&Multiply_Float(A=&MakeFloatFromVector2d(Vector 2D=$Particles.SpriteSize), B=0.5))
   &ScaleColor(
    Scale Alpha=&FloatFromCurve(FloatCurve=DataInterface() {propertyValues {Curve {keys [<at 0 0>,<at 0.005 1>]}}})
    ScaleA=true
    ScaleRGB=false
   ) {module=ScaleColor001}
  ]
  rendererClasses [&NiagaraSpriteRendererProperties]
  renderers [
   {rendererClass=&NiagaraSpriteRendererProperties}
   {rendererClass=&NiagaraSpriteRendererProperties, rendererIndex=1}
   {rendererClass=&NiagaraSpriteRendererProperties, rendererIndex=2}
   {rendererClass=&NiagaraSpriteRendererProperties, rendererIndex=3}
   {rendererClass=&NiagaraSpriteRendererProperties, rendererIndex=4}
   {rendererClass=&NiagaraSpriteRendererProperties, rendererIndex=5}
   {rendererClass=&NiagaraSpriteRendererProperties, rendererIndex=6}
  ]
 }
 Flash {
  emitterUpdateScript [
   &EmitterState()
   &Spawn_NDCFortImpacts(Reader:NiagaraDataInterfaceDataChannelRead=$System.NDCReader)
  ]
  particleSpawnScript [
   /SetVariables(Particles.Age:NiagaraFloat=$System.ImpactDelay) {isSetParameters=true}
   &InitializeParticleFromNDC_FortImpact(
    AmmoType Scale=1
    Distance Scale=1
    FallbackPosition:NiagaraPosition=<0 0 1>
    Scale Array=DataInterface() {propertyValues {GpuSyncMode=SyncCpuToGpu, MaxElements=4, FloatData [0.8,1,1.25,0.6,1]}}
   )
   /SetVariables() {isSetParameters=true}
   &InitializeParticle(
    Lifetime Max=0.2
    Lifetime Min=0.133
    Lifetime Mode=::NewEnumerator1
    Mesh Scale=&Multiply_Vector(A=<1.8 2.4 2.4>, B=&VectorFromFloat(Value=1.1))
    Mesh Scale Mode=::NewEnumerator4
    Position=$Particles.Position
    Position Mode=::NewEnumerator1
    Sprite Rotation Mode=::NewEnumerator2
    Sprite Size Mode=::NewEnumerator3
    Uniform Sprite Size=&Multiply_Float(A=&RandomRangeFloat(Minimum=130, Maximum=150), B=$Particles.AmmoTypeScale)
    UsePositionOffset=true
    Write Material Random=true
   )
   /SetVariables() {isSetParameters=true}
   &SubUVAnimation(End Frame=3, SubUV Animation Mode:ENiagara_SubUVLookupMode=::NewEnumerator2)
   &InitialMeshOrientation(
    Mesh Orientation Mode=::NewEnumerator2
    Orientation Vector=$Particles.ImpactSurfaceNormal
    Rotation=&MakeVector(
     X=&RandomRangeFloat(Minimum=0, Maximum=1)
     Y=&Multiply_Float(A=$Particles.AngleRange, B=&RandomRangeFloat(Minimum=-1, Maximum=1))
     Z=&Multiply_Float(A=$Particles.AngleRange, B=&RandomRangeFloat(Minimum=-1, Maximum=1))
    )
    Use Rotation Vector=true
   )
   /SetVariables(Particles.LightEnabled:NiagaraBool=&SetBoolByFloatComparison(
     A=$Particles.LODDistance
     B=&Multiply_Float(A=8000, B=$Particles.AmmoTypeScale)
     Comparison Type=::NewEnumerator7
    )) {isSetParameters=true}
   /SetVariables(Particles.LightPosition:NiagaraPosition=&AddVectorToPosition(
     Position=$Particles.Position
     Vector=&Multiply_VectorByFloat(Vector=$Particles.ImpactSurfaceNormal, Float=&Multiply_Float(A=20, B=$Particles.AmmoTypeScale))
    )) {isSetParameters=true}
   /SetVariables(Particles.LightVolumetricScattering=1) {isSetParameters=true}
   /SetVariables(Particles.LightRadius=&Multiply_Float(A=&Multiply_Float(A=90, B=$Particles.DistanceScale), B=$Particles.AmmoTypeScale)) {isSetParameters=true}
   /SetVariables(Particles.LightExponent=6) {isSetParameters=true}
   &CameraOffset(Camera Offset Amount=&Multiply_Float(A=20, B=$Particles.DistanceScale))
   &KillParticles(Kill Particles=&SetBoolByIntComparison(A=$Particles.PhysicalSurfaceInteger, B=8, Comparison Type=::NewEnumerator4))
  ]
  particleUpdateScript [
   &ParticleState()
   &Color(
    Color=&MakeLinearColorFromVectorAndFloat(
     "Vector (RGB)"=&MakeVectorFromLinearColorRGB(Color=&ColorFromCurve(ColorCurve=DataInterface() {propertyValues {
         AlphaCurve {keys []}
         BlueCurve {keys [{arriveTangent=-1.4036, interpMode=RCIM_Cubic, time=0.1241, value=0.0055}, <at 0.5798 0.0208>]}
         GreenCurve {keys [{arriveTangent=-5.857, interpMode=RCIM_Cubic, time=0.1241, value=7.0016}, <at 0.5798 0.8455>]}
         RedCurve {keys [{arriveTangent=-14.4701, interpMode=RCIM_Cubic, time=0.1241, value=10}, <at 0.5798 2>]}
        }}))
     "Float (Alpha)"=&FloatFromCurve(FloatCurve=DataInterface() {propertyValues {Curve {keys [
          <at 0 0 RCIM_Cubic>
          {arriveTangentWeight=0.0037, interpMode=RCIM_Cubic, leaveTangentWeight=0.369, tangentMode=RCTM_User, time=0.005, value=1}
          {arriveTangentWeight=0.2339, interpMode=RCIM_Cubic, leaveTangentWeight=0.2363, tangentMode=RCTM_User, time=0.2, value=1}
          {interpMode=RCIM_Cubic, tangentMode=RCTM_Break, time=1, value=0}
         ]}}})
    )
    Scale Color=&VectorFromFloat(Value=&FloatFromCurve(FloatCurve=DataInterface() {propertyValues {Curve {keys [{interpMode=RCIM_Cubic, tangentMode=RCTM_User, time=0, value=1}, {interpMode=RCIM_Cubic, tangentMode=RCTM_User, time=1, value=0}]}}}))
   )
   &ScaleMeshSize(Scale Factor=&Multiply_VectorByFloat(
     Vector=&VectorFromCurve(VectorCurve=DataInterface() {propertyValues {
        XCurve {keys [<at 0 1>,<at 1 1>]}
        YCurve {keys [{arriveTangent=1.6329, interpMode=RCIM_Cubic, time=0.1, value=0.2}, {interpMode=RCIM_Cubic, leaveTangent=1.0526, time=1, value=1}]}
        ZCurve {keys [{arriveTangent=1.6329, interpMode=RCIM_Cubic, time=0.1, value=0.2}, {interpMode=RCIM_Cubic, leaveTangent=1.0526, time=1, value=1}]}
       }})
     Float=&Multiply_Float(
      A=$Particles.DistanceScale
      B=&FloatFromCurve(FloatCurve=DataInterface() {propertyValues {Curve {keys [{arriveTangent=1.9165, interpMode=RCIM_Cubic, leaveTangent=1.9165, tangentMode=RCTM_User, time=0, value=0.25}, <at 1 1>]}}})
     )
    ))
   &FN_Sprite_Attributes(
    Create SpriteSize Attribute=true
    SpriteSize=&Multiply_Vector2D_ByFloat(
     Float=$Particles.DistanceScale
     Vector2D=&Multiply_Vector2D_ByFloat(
      Float=&FloatFromCurve(
       FloatCurve=DataInterface() {propertyValues {Curve {keys [
           {arriveTangent=1.126, interpMode=RCIM_Cubic, leaveTangent=1.126, tangentMode=RCTM_User, time=0, value=0.25}
           {arriveTangent=0.5184, tangentMode=RCTM_User, time=1, value=1}
          ]}}}
       Scale Curve=0.7
      )
      Vector2D=$Particles.Initial.SpriteSize
     )
    )
    Write to Module Namespace=true
   ) {module=FN_Sprite}
   /SetVariables(Particles.SpriteColor=&MakeLinearColorFromVectorAndFloat(
     "Vector (RGB)"=&MakeVectorFromLinearColorRGB(Color=&ColorFromCurve(ColorCurve=DataInterface() {propertyValues {
         AlphaCurve {keys []}
         BlueCurve {keys [{arriveTangent=-1.4036, interpMode=RCIM_Cubic, time=0.3167, value=0}, <at 0.8301 0>]}
         GreenCurve {keys [{arriveTangent=-5.857, interpMode=RCIM_Cubic, time=0.3167, value=2.611}, <at 0.8301 0.6032>]}
         RedCurve {keys [{arriveTangent=-14.4701, interpMode=RCIM_Cubic, time=0.3167, value=10}, <at 0.8301 2>]}
        }}))
     "Float (Alpha)"=&FloatFromCurve(FloatCurve=DataInterface() {propertyValues {Curve {keys [
          <at 0 0 RCIM_Cubic>
          {arriveTangentWeight=0.0037, interpMode=RCIM_Cubic, leaveTangentWeight=0.369, tangentMode=RCTM_User, time=0.005, value=1}
          {arriveTangentWeight=0.2339, interpMode=RCIM_Cubic, leaveTangentWeight=0.2363, tangentMode=RCTM_User, time=0.2, value=1}
          {interpMode=RCIM_Cubic, tangentMode=RCTM_Break, time=1, value=0}
         ]}}})
    )) {isSetParameters=true}
   /SetVariables(Particles.Light_Color:LinearColor=&MultiplyLinearColor(
     Linear Color=&LerpLinearColors(
      StartColor=<rgba 5 3.5825 0 1>
      EndColor=<rgba 2 1.0012 0.0035 1>
      LerpFactor=&FloatFromCurve(FloatCurve=DataInterface() {propertyValues {Curve {keys [<at 0.4 0 RCIM_Cubic>,<at 0.6 1 RCIM_Cubic>]}}})
     )
     Scale Factor=&MakeFloat4FromVectorAndFloat(
      "Vector (RGB)"=<1 1 1>
      "Float (Alpha)"=&Multiply_Float(
       A=15
       B=&FloatFromCurve(
        FloatCurve=DataInterface() {propertyValues {Curve {keys [
            {interpMode=RCIM_Cubic, tangentMode=RCTM_User, time=0, value=0}
            {interpMode=RCIM_Cubic, tangentMode=RCTM_User, time=0.2, value=1}
            {tangentMode=RCTM_User, time=1, value=0}
           ]}}}
        Scale Curve=&FloatFromCurve(
         FloatCurve=DataInterface() {propertyValues {Curve {keys [{interpMode=RCIM_Cubic, tangentMode=RCTM_User, time=0, value=1}, {interpMode=RCIM_Cubic, tangentMode=RCTM_User, time=1, value=0}]}}}
         CurveIndex=&RemapRange(Input Value To Remap=$Particles.LODDistance, Input Min=0, Input Max=2000, Output Min=0, Output Max=1)
        )
       )
      )
     )
    )) {isSetParameters=true}
  ]
  rendererClasses [
   &NiagaraMeshRendererProperties
   &NiagaraSpriteRendererProperties
   &NiagaraLightRendererProperties
  ]
  renderers [
   {rendererClass=&NiagaraMeshRendererProperties, rendererIndex=0}
   {rendererClass=&NiagaraSpriteRendererProperties, rendererIndex=1}
   {rendererClass=&NiagaraLightRendererProperties, rendererIndex=2}
  ]
 }
 NE_FN_PhysMatDebris {
  emitterUpdateScript [
   &EmitterState()
   &Spawn_NDCFortImpacts(
    Distance Culling=1
    LOD Distance=5000
    Max Spawn Count=15
    Min Spawn Count=5
    Reader:NiagaraDataInterfaceDataChannelRead=$System.NDCReader
    Simple Max Spawn Count=20
    Simple Min Spawn Count=5
    Simple Spawn Override=2
    Spawn Count Mode=2
   )
  ]
  particleSpawnScript [
   /SetVariables(Particles.Age:NiagaraFloat=$System.ImpactDelay) {isSetParameters=true}
   &InitializeParticleFromNDC_FortImpact(
    AmmoType Scale=1
    Distance Scale=1
    FallbackPosition:NiagaraPosition=<0 0 1>
    Scale Array=DataInterface() {propertyValues {GpuSyncMode=SyncCpuToGpu, MaxElements=4, FloatData [0.8,1,1.2,0.6,1]}}
   )
   /SetVariables(Particles.IsWater:NiagaraBool=&SetBoolByIntComparison(A=$Particles.PhysicalSurfaceInteger, B=8, Comparison Type=::NewEnumerator4)) {isSetParameters=true}
   /SetVariables(Particles.BiasedRandom=&FloatFromCurve(
     FloatCurve=DataInterface() {propertyValues {Curve {keys [
         <at 0.2 1 RCIM_Cubic>
         {arriveTangent=-3.3394, interpMode=RCIM_Cubic, leaveTangent=-3.3394, tangentMode=RCTM_User, time=0.5, value=0.5}
         <at 0.8 0 RCIM_Cubic>
        ]}}}
     CurveIndex=&RandomRangeFloat(Minimum=0, Maximum=1)
    )) {isSetParameters=true}
   &InitializeParticle(
    Color Mode=::NewEnumerator0
    Lifetime=&MakeCustomFloatFromBool(
     Bool=$Particles.IsWater
     True Float=&RandomRangeFloat(Minimum=0.5, Maximum=1)
     False Float=&RandomRangeFloat(Minimum=0.5, Maximum=1)
    )
    Mass=&Lerp_Float(A=0.6, B=1.2, Alpha=$Particles.BiasedRandom)
    Mass Mode=::NewEnumerator1
    Mesh Scale=&Multiply_VectorByFloat(
     Vector=&RandomRangeVector(Minimum=<0.9 0.4 0.5>, Maximum=<1 1 1>)
     Float=&Lerp_Float(A=0.7, B=1, Alpha=$Particles.BiasedRandom)
    )
    Mesh Scale Mode=::NewEnumerator4
    Position=$Particles.Position
    Position Mode=::NewEnumerator1
    Sprite Size Mode=::NewEnumerator1
    Sprite UV Mode=::NewEnumerator2
    UsePositionOffset=false
    Write Material Random=true
   )
   /SetVariables(Particles.IsSnow:NiagaraBool=&SetBoolByIntComparison(A=$Particles.PhysicalSurfaceInteger, B=23, Comparison Type=::NewEnumerator4)) {isSetParameters=true}
   &FN_PhysicalSurface(
    Dirt Mesh Count=2
    Dirt Scale:NiagaraFloat=&RandomRangeFloat(Minimum=0.33, Maximum=1)
    Dirt Velocity:NiagaraFloat=&RandomRangeFloat(Minimum=150, Maximum=400)
    Glass Mesh Count=3
    Glass Scale=0.5
    Glass Velocity:NiagaraFloat=&RandomRangeFloat(Minimum=100, Maximum=400)
    Glass Velocity Cone Angle:NiagaraFloat=&RandomRangeFloat(Minimum=20, Maximum=70)
    Grass Mass=0.25
    Grass Mesh Count=2
    Grass Scale:NiagaraFloat=&RandomRangeFloat(Minimum=0.66, Maximum=1)
    Grass Velocity:NiagaraFloat=&RandomRangeFloat(Minimum=150, Maximum=400)
    Grass Velocity Cone Angle=30
    Ice Mesh Count=2
    Ice Scale:NiagaraFloat=&MakeCustomFloatFromBool(
     Bool=$Particles.IsSnow
     True Float=&RandomRangeFloat(Minimum=0.8, Maximum=1)
     False Float=&RandomRangeFloat(Minimum=0.33, Maximum=0.5)
    )
    Ice Velocity:NiagaraFloat=&MakeCustomFloatFromBool(
     Bool=$Particles.IsSnow
     True Float=&RandomRangeFloat(Minimum=100, Maximum=300)
     False Float=&RandomRangeFloat(Minimum=100, Maximum=500)
    )
    Ice Velocity Cone Angle:NiagaraFloat=&MakeCustomFloatFromBool(
     Bool=$Particles.IsSnow
     True Float=&RandomRangeFloat(Minimum=50, Maximum=90)
     False Float=&RandomRangeFloat(Minimum=60, Maximum=120)
    )
    Ice Visibility:NiagaraInt32=&MakeCustomIntFromBool(Bool=$Particles.IsSnow, True INT=23, False INT=24)
    Metal Gravity=&RandomRangeFloat(Minimum=-1400, Maximum=-1800)
    Metal Mesh Count=2
    Metal Velocity=&RandomRangeFloat(Minimum=80, Maximum=600)
    Metal Velocity Cone Angle=30
    Physical Surface=&SelectIntBySurface(
     Default:NiagaraInt32=$Particles.PhysicalSurfaceInteger
     Surface:EPhysicalSurface=$Particles.PhysicalSurface
     Num Surfaces:NiagaraInt32=3
     Surface 0:EPhysicalSurface=SurfaceType_Default
     Value For Surface 0:NiagaraInt32=2
     Surface 1:EPhysicalSurface=SurfaceType23
     Value For Surface 1:NiagaraInt32=24
     Surface 2:EPhysicalSurface=SurfaceType9
     Value For Surface 2:NiagaraInt32=&MakeCustomIntFromBool(Bool=&RandomBool(), True INT=9, False INT=10)
     Surface 3:EPhysicalSurface=?
     Value For Surface 3:NiagaraInt32=?
    )
    Stone Mesh Count=3
    Stone Velocity=&RandomRangeFloat(Minimum=150, Maximum=400)
    Stone Velocity Cone Angle:NiagaraFloat=&RandomRangeFloat(Minimum=30, Maximum=120)
    StoneScale:NiagaraFloat=&RandomRangeFloat(Minimum=0.33, Maximum=1)
    Wood Mesh Count=2
    Wood Scale:NiagaraFloat=&RandomRangeFloat(Minimum=0.75, Maximum=1.5)
    Wood Velocity=&RandomRangeFloat(Minimum=80, Maximum=500)
    Wood Velocity Cone Angle=60
   )
   &AddVelocityInCone(
    Cone Angle=$Output.FN_PhysicalSurface.SurfaceResponseConeAngle
    Cone Axis=$Particles.ImpactSurfaceNormal
    Cone Axis Coordinate Space=::Simulation
    Module.Use Velocity Falloff On Cone Axis=false
    Velocity Strength=&Multiply_Float(A=$Output.FN_PhysicalSurface.SurfaceResponseVelocity, B=$Particles.AmmoTypeScale)
   )
   /SetVariables(Particles.Velocity=&Multiply_VectorByFloat(
     Vector=$Particles.Velocity
     Float=&Lerp_Float(A=0.5, B=2, Alpha=&ReturnNormalizedExecIndex())
    )) {isSetParameters=true}
   &AddVelocity(
    Rotation Coordinate Space=::World
    Velocity=<0 0 1>
    Velocity Speed Scale=&Multiply_Float(A=250, B=$Particles.AmmoTypeScale)
   )
   &InitialMeshOrientation(Use Rotation Vector=false)
   &AddRotationalVelocity(Rotation Rate=&RandomVector(Vector Scale=&RandomRangeFloat(Minimum=1, Maximum=2)))
   /SetVariables(Particles.GravityOverrideValue=$Output.FN_PhysicalSurface.GravityOverride) {isSetParameters=true}
   &AddVelocity(
    Cone Angle=40
    Cone Axis=<0 0 1>
    Module.UseConeFalloff=false
    Module.UseVelDistribution=false
    Override Seed=false
    Velocity Mode=::NewEnumerator2
    Velocity Speed=&MakeCustomFloatFromBool(
     Bool=$Particles.IsWater
     True Float=&RandomRangeFloat(Minimum=250, Maximum=750)
     False Float=&RandomRangeFloat(Minimum=0, Maximum=150)
    )
   ) {module=AddVelocity002}
   &ApplyInitialForces()
  ]
  particleUpdateScript [
   &ParticleState()
   &ScaleSpriteSize(Uniform Curve Sprite Scale=DataInterface() {propertyValues {Curve {keys [{interpMode=RCIM_Cubic, leaveTangent=-2.6436, tangentMode=RCTM_Break, time=0, value=1}, {tangentMode=RCTM_User, time=1, value=0}]}}})
   &Drag(
    Drag=&RandomRangeFloat(Minimum=1, Maximum=2)
    Rotational Drag=0.1
    Use Linear Drag=true
    Use Rotational Drag=true
   )
   &GravityForce(Gravity=&Multiply_VectorByFloat(Vector=$Particles.GravityOverrideValue, Float=2))
   &SolveForcesAndVelocity(
    Clamp Velocity=false
    DeltaTime:NiagaraFloat=&MakeCustomFloatFromBool(Bool=&SetBoolByFloatComparison(A=$Particles.Age, B=0), True Float=$Engine.DeltaTime, False Float=0)
    Limit Acceleration=false
    Manually Enable Rotational Solver=true
   )
   &ScaleMeshSize(Scale Factor=&VectorFromFloat(Value=&FloatFromCurve(
      FloatCurve=DataInterface() {propertyValues {Curve {keys [<at 0.6 1 RCIM_Cubic>, {arriveTangent=-6.308, interpMode=RCIM_Cubic, leaveTangent=-6.308, tangentMode=RCTM_User, time=1, value=0}]}}}
      Scale Curve=&Lerp_Float(A=$Particles.AmmoTypeScale, B=$Particles.DistanceScale, Alpha=0.5)
     )))
   &DynamicMaterialParameters(Module.Index0Param0Write=true, Module.Index0Param1Write=true, Module.Index0Param2Write=true, Module.Index0Param3Write=true)
   /SetVariables(Particles.Splash Color:LinearColor=&ColorFromCurve(ColorCurve=DataInterface() {propertyValues {
       AlphaCurve {keys [<at 0 0.75 RCIM_Cubic>,<at 0.9721 0>]}
       BlueCurve {keys [<at 0 1 RCIM_Cubic>,<at 1 1 RCIM_Cubic>]}
       GreenCurve {keys [<at 0 1 RCIM_Cubic>,<at 1 1 RCIM_Cubic>]}
       RedCurve {keys [<at 0 1 RCIM_Cubic>,<at 1 1 RCIM_Cubic>]}
      }})) {isSetParameters=true}
   /SetVariables(Particles.SpriteSize_Splash:Vector2f=&Multiply_Vector2D_ByFloat(
     Float=&FloatFromCurve(FloatCurve=DataInterface() {propertyValues {Curve {keys [{interpMode=RCIM_Cubic, leaveTangent=3.7456, tangentMode=RCTM_Break, time=0, value=0.125}, {tangentMode=RCTM_User, time=0.7, value=1.05}]}}})
     Vector2D=&RandomRangeVector2D(Minimum=<50 50>, Maximum=<150 150>)
    )) {isSetParameters=true}
  ]
  rendererClasses [&NiagaraMeshRendererProperties]
  renderers [
   {rendererClass=&NiagaraMeshRendererProperties, rendererIndex=0}
   {rendererClass=&NiagaraMeshRendererProperties, rendererIndex=1}
   {rendererClass=&NiagaraMeshRendererProperties, rendererIndex=2}
   {rendererClass=&NiagaraMeshRendererProperties, rendererIndex=3}
   {rendererClass=&NiagaraMeshRendererProperties, rendererIndex=4}
   {rendererClass=&NiagaraMeshRendererProperties, rendererIndex=5}
   {rendererClass=&NiagaraMeshRendererProperties, rendererIndex=6}
  ]
 }
 Sparks001 {
  emitterUpdateScript [
   &EmitterState()
   &Spawn_NDCFortImpacts(Max Spawn Count=40, Min Spawn Count=15, Spawn Count Mode=2)
  ]
  particleSpawnScript [
   /SetVariables(Particles.Age:NiagaraFloat=$System.ImpactDelay) {isSetParameters=true}
   &self:Initialize_Particle_from_NDC_05() {module=Initialize_Particle_from_NDC_05}
   /SetVariables(Particles.ScaleLerp:NiagaraFloat=&FloatFromCurve(FloatCurve=DataInterface() {propertyValues {Curve {keys [<at 1000 0>,<at 50000 1>]}}}, CurveIndex=$Particles.LODDistance)) {isSetParameters=true}
   &InitializeParticle(
    Color Maximum=<rgba 5 3.69 0.013 1>
    Color Minimum=<rgba 10 4.5087 0.0155 0.4>
    Color Mode=::NewEnumerator2
    Lifetime=&FloatFromCurve(
     FloatCurve=DataInterface() {propertyValues {Curve {
        defaultValue=1
        keys [
         {arriveTangent=0.4191, interpMode=RCIM_Cubic, leaveTangent=0.4191, tangentMode=RCTM_User, time=0, value=0.1}
         {arriveTangent=2.3656, interpMode=RCIM_Cubic, leaveTangent=2.3656, tangentMode=RCTM_User, time=1, value=1}
        ]
       }}}
     CurveIndex=&RandomRangeFloat(Minimum=0, Maximum=0.5)
    )
    Position=$Particles.Position
    Position Mode=::NewEnumerator1
    Sprite Size Mode=::NewEnumerator3
    Uniform Sprite Size=&Lerp_Float(A=1, B=2, Alpha=&Lerp_Float(A=0.005, B=3, Alpha=$Particles.ImpactScale))
    UsePositionOffset=false
    Write Material Random=true
   )
   &ShapeLocation(MaskSpawnGroup=false, Override Seed=false, Sphere Radius=10, UseOwnerScale=true)
   &AddVelocity(
    Cone Angle=30
    Cone Axis=$Particles.ImpactSurfaceNormal
    Inner Cone Angle=180
    Module.UseConeFalloff=false
    Module.UseVelDistribution=false
    Override Seed=false
    Speed Falloff From Cone Axis=0.2
    Velocity Mode=::NewEnumerator2
    Velocity Speed=&Lerp_Float(
     A=500
     B=2000
     Alpha=&FloatFromCurve(
      FloatCurve=DataInterface() {propertyValues {Curve {keys [
          {arriveTangent=0.2237, interpMode=RCIM_Cubic, leaveTangent=0.2237, tangentMode=RCTM_User, time=0, value=0}
          {arriveTangent=2.6808, interpMode=RCIM_Cubic, leaveTangent=2.6808, tangentMode=RCTM_User, time=1, value=1}
         ]}}}
      CurveIndex=&RandomRangeFloat(Minimum=0, Maximum=1)
     )
    )
   ) {enabled=false}
   &AddVelocity(
    Cone Angle=170
    Cone Axis=$Particles.ImpactSurfaceNormal
    Module.UseConeFalloff=true
    Module.UseVelDistribution=true
    Override Seed=false
    Velocity Mode=::NewEnumerator2
    Velocity Speed=&Multiply_Float(
     A=&Lerp_Float(
      A=2000
      B=8000
      Alpha=&FloatFromCurve(
       FloatCurve=DataInterface() {propertyValues {Curve {keys [
           {arriveTangent=0.3512, interpMode=RCIM_Cubic, leaveTangent=0.3512, tangentMode=RCTM_User, time=0, value=0}
           {arriveTangent=2.6808, interpMode=RCIM_Cubic, leaveTangent=2.6808, tangentMode=RCTM_User, time=1, value=1}
          ]}}}
       CurveIndex=&RandomRangeFloat(Minimum=0, Maximum=1)
      )
     )
     B=0.7
    )
   ) {module=AddVelocity002}
   &AddVelocity(
    Cone Angle=90
    Cone Axis=$Particles.ImpactSurfaceNormal
    Module.UseConeFalloff=false
    Module.UseVelDistribution=false
    Override Seed=false
    Velocity Mode=::NewEnumerator2
    Velocity Speed=&RandomRangeFloat(Minimum=200, Maximum=1800)
   ) {enabled=false, module=AddVelocity001}
  ]
  particleUpdateScript [
   &ParticleState()
   &GravityForce(Gravity=&RandomRangeVector(Minimum=<0 0 -1500>, Maximum=<0 0 -2000>))
   &Drag(
    Drag=&FloatFromCurve(
     FloatCurve=DataInterface() {propertyValues {Curve {keys [
         {interpMode=RCIM_Cubic, leaveTangent=-8.9049, leaveTangentWeight=0.5018, tangentMode=RCTM_Break, time=0, value=1}
         {interpMode=RCIM_Cubic, tangentMode=RCTM_Break, time=0.3, value=0.025}
        ]}}}
     Scale Curve=&RandomRangeFloat(Minimum=5, Maximum=12)
    )
    Use Linear Drag=true
    Use Rotational Drag=false
   ) {enabled=false}
   &Drag(
    Drag=&FloatFromCurve(
     FloatCurve=DataInterface() {propertyValues {Curve {
        defaultValue=1
        keys [
         {arriveTangent=2.8935, interpMode=RCIM_Cubic, leaveTangent=2.8935, tangentMode=RCTM_User, time=0, value=0}
         {tangentMode=RCTM_User, time=1, value=1}
        ]
       }}}
     Scale Curve=&RandomRangeFloat(Minimum=4, Maximum=6)
    )
    Use Linear Drag=true
    Use Rotational Drag=false
   ) {module=Drag001}
   &VectorNoiseForce(
    Force Amount=&FloatFromCurve(
     FloatCurve=DataInterface() {propertyValues {Curve {
        defaultValue=1
        keys [{interpMode=RCIM_Cubic, tangentMode=RCTM_User, time=0, value=0}, {interpMode=RCIM_Cubic, tangentMode=RCTM_User, time=0.5, value=1}]
       }}}
     Scale Curve=12000
    )
    Module.Use Random Noise Time Interval=true
    Random Noise Time Interval:NiagaraFloat=&RandomRangeFloat(Minimum=0.2, Maximum=0.3)
   )
   &SolveForcesAndVelocity(
    Clamp Velocity=false
    DeltaTime:NiagaraFloat=&MakeCustomFloatFromBool(Bool=&SetBoolByFloatComparison(A=$Particles.Age, B=0), True Float=$Engine.DeltaTime, False Float=0)
    Limit Acceleration=false
   )
   &ScaleSpriteSize(
    Uniform Curve Scale=&Multiply_Float(A=0.3, B=1)
    Uniform Curve Sprite Scale=DataInterface() {propertyValues {Curve {keys [
        {interpMode=RCIM_Cubic, tangentMode=RCTM_Break, time=0.25, value=1}
        {arriveTangent=-2.298, interpMode=RCIM_Cubic, tangentMode=RCTM_Break, time=1, value=0.2}
       ]}}}
   )
   &ScaleSpriteSizeBySpeed(
    Max Scale Factor=<1 10>
    Sample Scale Factor By Curve=true
    Scale Factor Curve=DataInterface() {propertyValues {Curve {keys [{arriveTangent=0.3851, interpMode=RCIM_Cubic, time=0, value=0}, {arriveTangent=2.2117, tangentMode=RCTM_User, time=1, value=1}]}}}
    Velocity Threshold=1500
   )
   &ScaleColor(
    Scale Alpha=&FloatFromCurve(FloatCurve=DataInterface() {propertyValues {Curve {keys [<at 0 1 RCIM_Cubic>, {arriveTangent=-2.9021, tangentMode=RCTM_User, time=1, value=0}]}}})
    ScaleA=true
    ScaleRGB=false
   ) {module=ScaleColor001}
   &DynamicMaterialParameters(
    Index 0 Param 1=&FloatFromCurve(
     FloatCurve=DataInterface() {propertyValues {Curve {defaultValue=1, keys [<at 0 0 RCIM_Cubic>, {arriveTangent=1.1412, tangentMode=RCTM_User, time=1, value=1}]}}}
     Scale Curve=2
    )
    Index 0 Param 3=3
    Module.Index0Float0Write=true
    Module.Index0Float1Write=false
    Module.Index0Float2Write=true
    Module.Index0Float3Write=false
   )
   &ScaleColor(
    Scale Alpha=&FloatFromCurve(FloatCurve=DataInterface() {propertyValues {Curve {keys [<at 0 0>,<at 0.005 1>]}}})
    ScaleA=true
    ScaleRGB=false
   )
  ]
  rendererClasses [&NiagaraSpriteRendererProperties]
  renderers [{rendererClass=&NiagaraSpriteRendererProperties}]
 }
 SpikeyFlash {
  emitterUpdateScript [
   &EmitterState()
   &Spawn_NDCFortImpacts()
  ]
  particleSpawnScript [
   /SetVariables(Particles.Age:NiagaraFloat=$System.ImpactDelay) {isSetParameters=true}
   &self:Initialize_Particle_from_NDC_02() {module=Initialize_Particle_from_NDC_02}
   &InitializeParticle(
    Color=<rgba 5 3.123 0 1>
    Lifetime=&Lerp_Float(
     A=&RandomRangeFloat(Minimum=0.14, Maximum=0.17)
     B=&RandomRangeFloat(Minimum=0.18, Maximum=0.24)
     Alpha=$Particles.ImpactScale
    )
    Mesh Scale=&Multiply_VectorByFloat(
     Vector=&RandomRangeVector(Minimum=<0.05 0.05 0.25>, Maximum=<0.07 0.07 0.5>)
     Float=&Lerp_Float(A=0.5, B=1, Alpha=$Particles.ImpactScale)
    )
    Mesh Scale Mode=::NewEnumerator4
    Position=$Particles.Position
    Position Mode=::NewEnumerator1
    Sprite Rotation Mode=::NewEnumerator2
    Sprite Size Mode=::NewEnumerator3
    Uniform Sprite Size=&Multiply_Float(A=300, B=&Lerp_Float(A=0.25, B=1, Alpha=$Particles.ImpactScale))
    UsePositionOffset=false
    Write Material Random=true
   )
   &CameraOffset(Camera Offset Amount=50)
   &InitialMeshOrientation(
    Mesh Axis To Orient=<0 0 1>
    Mesh Orientation Mode=::NewEnumerator2
    Orientation Vector=$Particles.ImpactSurfaceNormal
    Rotation=&RandomRangeVector(Minimum=<0 0 0>, Maximum=<0 0 1>)
    Use Rotation Vector=true
   )
   &SpriteFacingAndAlignment(Sprite Facing=$Particles.ImpactSurfaceNormal, WriteAlignment=false, WriteFacing=true)
  ]
  particleUpdateScript [
   &ParticleState()
   &ScaleSpriteSize(
    Uniform Curve Scale=0.5
    Uniform Curve Sprite Scale=DataInterface() {propertyValues {Curve {keys [
        {arriveTangent=2.8769, interpMode=RCIM_Cubic, leaveTangent=2.8769, tangentMode=RCTM_User, time=0, value=0}
        {tangentMode=RCTM_User, time=1, value=1}
       ]}}}
   )
   &SolveForcesAndVelocity(Clamp Velocity=false, Limit Acceleration=false)
   &ScaleMeshSize(Scale Factor=&VectorFromFloat(Value=&FloatFromCurve(
      FloatCurve=DataInterface() {propertyValues {Curve {keys [
          {arriveTangent=9.901, interpMode=RCIM_Cubic, leaveTangent=9.901, tangentMode=RCTM_User, time=0, value=0.1}
          {arriveTangent=2.6083, interpMode=RCIM_Cubic, leaveTangent=2.6083, tangentMode=RCTM_User, time=1, value=6}
         ]}}}
      Scale Curve=0.3
     )))
   &ScaleColor(
    Scale Alpha=&FloatFromCurve(FloatCurve=DataInterface() {propertyValues {Curve {keys [<at 0.4 1 RCIM_Cubic>, {arriveTangent=-2.2027, tangentMode=RCTM_User, time=1, value=0}]}}})
    Scale RGB=&MakeVectorFromLinearColorRGB(Color=&LerpLinearColors(StartColor=<rgba 1 1 1 1>, EndColor=<rgba 0 0 0 1>, LerpFactor=0))
    ScaleA=true
    ScaleRGB=true
   )
   /SetVariables(Particles.Glow Sprite Color:LinearColor=&MakeLinearColorFromVectorAndFloat(
     "Vector (RGB)"=&MakeVectorFromLinearColorRGB(Color=&LerpLinearColors(
       StartColor=<rgba 2 0.7401 0 1>
       EndColor=<rgba 1 0.3887 0 1>
       LerpFactor=&FloatFromCurve(FloatCurve=DataInterface() {propertyValues {Curve {keys [<at 0.35 0 RCIM_Cubic>,<at 1 1 RCIM_Cubic>]}}})
      ))
     "Float (Alpha)"=&FloatFromCurve(
      FloatCurve=DataInterface() {propertyValues {Curve {keys [
          {arriveTangent=-2.7319, interpMode=RCIM_Cubic, time=0.05, value=1}
          {arriveTangent=-1.5377, interpMode=RCIM_Cubic, leaveTangent=-1.5377, tangentMode=RCTM_User, time=0.2, value=0.55}
          <at 1 0>
         ]}}}
      Scale Curve=5
     )
    )) {isSetParameters=true}
  ]
  rendererClasses [&NiagaraSpriteRendererProperties]
  renderers [
   {rendererClass=&NiagaraSpriteRendererProperties}
   {rendererClass=&NiagaraSpriteRendererProperties, rendererIndex=1}
  ]
 }
 Streaks_Fat {
  emitterUpdateScript [
   &EmitterState()
   &Spawn_NDCFortImpacts(Max Spawn Count=5, Min Spawn Count=4, Spawn Count Mode=2)
  ]
  particleSpawnScript [
   /SetVariables(Particles.Age:NiagaraFloat=$System.ImpactDelay) {isSetParameters=true}
   &self:Initialize_Particle_from_NDC_01() {module=Initialize_Particle_from_NDC_01}
   &InitializeParticle(
    Lifetime Max=0.15
    Lifetime Min=0.11
    Lifetime Mode=::NewEnumerator1
    Position=$Particles.Position
    Position Mode=::NewEnumerator1
    Sprite Size=&Multiply_Vector2D_ByFloat(Float=1, Vector2D=&RandomRangeVector2D(Minimum=<8 50>, Maximum=<12 100>))
    Sprite Size Mode=::NewEnumerator4
    UsePositionOffset=false
    Write Material Random=true
   )
   &ShapeLocation(MaskSpawnGroup=false, Override Seed=false, Sphere Radius=1, Sphere Surface Distribution=0.5595, UseOwnerScale=true)
   &AddVelocity(
    Cone Angle=10
    Cone Axis=$Particles.ImpactSurfaceNormal
    Inner Cone Angle=90
    Module.UseConeFalloff=false
    Module.UseVelDistribution=false
    Override Seed=false
    Velocity Mode=::NewEnumerator2
    Velocity Speed=&Lerp_Float(
     A=0
     B=1
     Alpha=&FloatFromCurve(
      FloatCurve=DataInterface() {propertyValues {Curve {keys [
          {interpMode=RCIM_Cubic, tangentMode=RCTM_User, time=0, value=0}
          {arriveTangent=2.6808, interpMode=RCIM_Cubic, leaveTangent=2.6808, tangentMode=RCTM_User, time=1, value=1}
         ]}}}
      CurveIndex=&RandomRangeFloat(Minimum=0, Maximum=1)
     )
    )
   )
  ]
  particleUpdateScript [
   &ParticleState()
   &SolveForcesAndVelocity(Clamp Velocity=false, Limit Acceleration=false)
   &ScaleSpriteSize(Scale Factor=&Vector2DFromCurve(
     Vector2Curve=DataInterface() {propertyValues {XCurve {keys [<at 0 0.75 RCIM_Cubic>,<at 1 1 RCIM_Cubic>]}, YCurve {keys [<at 0 0 RCIM_Cubic>,<at 1 1.6 RCIM_Cubic>]}}}
     ScaleCurve=<1 1>
    ))
   &ScaleColor(
    Scale Alpha=&FloatFromCurve(FloatCurve=DataInterface() {propertyValues {Curve {keys [<at 0 0.9915 RCIM_Cubic>,<at 1 0>]}}})
    Scale RGB=&Multiply_Vector(
     A=&MakeVectorFromLinearColorRGB(Color=&ColorFromCurve(ColorCurve=DataInterface() {propertyValues {
         AlphaCurve {keys [<at 0 1 RCIM_Cubic>,<at 1 0 RCIM_Cubic>]}
         BlueCurve {keys [<at 0.2957 0 RCIM_Cubic>,<at 0.7103 0.0007>]}
         GreenCurve {keys [<at 0.2957 1.5484 RCIM_Cubic>,<at 0.7103 0.9685>]}
         RedCurve {keys [<at 0.2957 2 RCIM_Cubic>,<at 0.7103 2>]}
        }}))
     B=&VectorFromFloat(Value=&FloatFromCurve(FloatCurve=DataInterface() {propertyValues {Curve {keys [<at 0 3 RCIM_Cubic>,<at 0.5 1 RCIM_Cubic>]}}}, Scale Curve=1.2))
    )
    ScaleA=true
    ScaleRGB=true
   )
  ]
  rendererClasses [&NiagaraSpriteRendererProperties]
  renderers [{rendererClass=&NiagaraSpriteRendererProperties}]
 }
}
summary {
 systemName=NS_NDC_Handler_Impacts_GoldenGun
 emitters [
  {
   emitterName=Flash
   rendererClasses [
    &NiagaraMeshRendererProperties
    &NiagaraSpriteRendererProperties
    &NiagaraLightRendererProperties
   ]
  }
  {emitterName=NE_FN_PhysMatDebris, rendererClasses [&NiagaraMeshRendererProperties]}
  {emitterName=Flames, rendererClasses [&NiagaraSpriteRendererProperties]}
  {emitterName=Streaks_Fat, rendererClasses [&NiagaraSpriteRendererProperties]}
  {emitterName=SpikeyFlash, rendererClasses [&NiagaraSpriteRendererProperties]}
  {emitterName=Sparks001, rendererClasses [&NiagaraSpriteRendererProperties]}
  {emitterName=DecalRenderer, rendererClasses [&NiagaraDecalRendererProperties]}
 ]
 userVariables [User.HitNormal {type=Vector3f, description="", defaultValue {struct=Vector3f, value=<0 0 1>}}]
}
