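# Niagara system that draws Golden Gun projectile effects, reading projectile events from a shared Niagara Data Channel.
# Bullet draws gold, hue-shiftable tracer sprites, while instigator-filtered Trail and Local_Player_Trail emitters draw smoke-style ribbons.
# Sparks scatters noisy sprites randomly along each shot segment, using the local-trail filter.
# Hitscan detection remaps age curves that lerp tracers and trails between previous and end positions, with ammo-type and LOD-distance scaling.
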NiagaraSystem /WeaponsUnvaulted/Assets/VFX/GoldenGun/NS_NDC_ProjectilesHandler_GoldenGun
dependencies {
 usedDataInterfaces [
  &NiagaraDataInterfaceArrayFloat
  &NiagaraDataInterfaceCurve
  /Script/Niagara.NiagaraDataInterfaceDataChannelRead
 ]
 usedDynamicInputs []
 usedModules [
  &self:SystemSpawnScript.SystemScriptSource.SystemScriptGraph.NiagaraNodeAssignment_1.SetVariables_#1
  &self:SystemSpawnScript.SystemScriptSource.SystemScriptGraph.NiagaraNodeAssignment_2.SetVariables_#2
  &SystemState
  /Niagara/Modules/System/CompleteIfUnused
  &self:SystemSpawnScript.SystemScriptSource.SystemScriptGraph.NiagaraNodeAssignment_0.SetVariables_#3
 ]
}
emitters {
 Bullet {
  emitterUpdateScript [
   &EmitterState()
   /Game/Effects/Niagara/DataChannels/Projectiles/Spawn_NDCFortProjectiles(Reader:NiagaraDataInterfaceDataChannelRead=$System.NDCReader)
  ]
  particleSpawnScript [
   /Game/Effects/Niagara/DataChannels/Projectiles/InitializeParticleFromNDC_FortProjectiles(
    End Lifetime=true
    End Lifetime Min=0.05
    FallbackAmmoType=2
    FallbackOrigin:NiagaraPosition=<0 50 0>
    Intro Time=true
    LOD Output Scale Min=0.5
    Length Scale Array=DataInterface() {propertyValues {GpuSyncMode=SyncCpuToGpu, MaxElements=4, FloatData [0.75,0.85,1,0.7]}}
    Length Scale by Array=true
    Length by Velocity=true
    Lifetime=0.0333
    Scale Array=DataInterface() {propertyValues {GpuSyncMode=SyncCpuToGpu, MaxElements=4, FloatData [8,12,16,6]}}
    Scale by Array=true
    Use LOD Distance Scaling=true
   )
   &KillParticles(Kill Particles=$Particles.IsStartPosition) {module=KillParticles001}
   &InitializeParticle(
    Color Mode=::NewEnumerator0
    Lifetime=$Particles.Lifetime
    Position=$Particles.CurrentPosition
    Position Mode=::NewEnumerator1
    Sprite Size=&MakeVector2D(
     X=$Particles.ProjectileScale
     Y=&Multiply_Float(A=$Particles.ProjectileLength, B=$Particles.NormalizedIntroDuration)
    )
    Sprite Size Mode=::NewEnumerator4
    UsePositionOffset=false
    Write Material Random=true
   )
   &DynamicMaterialParameters(
    Index 0 Param 1=2
    Index 0 Param 2=1
    Index 0 Param 3=3
    Index 0 Param 4=0.25
    Module.Index0Param0Write=true
    Module.Index0Param1Write=true
    Module.Index0Param2Write=true
    Module.Index0Param3Write=true
   )
   /SetVariables(Particles.Color=&self:Material_Hue_Shift_01(
     Color:LinearColor=&MakeCustomLinearColorFromBool(Bool=$Particles.IsEndPosition, True Color=<rgba 2 0.5 0.05 0>, False Color=<rgba 2 0.5 0.05 1>)
     Hue Shift:NiagaraFloat=$Particles.HueShift
    )) {isSetParameters=true, module=SetVariables_#4}
   /SetVariables(Particles.Color=&MakeCustomLinearColorFromBool(Bool=$Particles.IsEndPosition, True Color=<rgba 2 0.5 0.05 0>, False Color=<rgba 2 0.5 0.05 1>)) {enabled=false, isSetParameters=true}
   /SetVariables(Particles.IsHitScan:NiagaraBool=&Boolean_AndOperation(
     Bool1=$Particles.IsEndPosition
     Bool2=&SetBoolByFloatComparison(
      A=&VectorLength(
       Vector=&SubtractVectorFromPosition(Position=$Particles.PreviousPosition, Vector=Linked() {linkedVariable {name=Particles.OriginPosition, type=NiagaraPosition}})
       Safe Length Calculation=false
      )
      B=0
      Comparison Type=::NewEnumerator4
     )
    )) {isSetParameters=true, module=SetVariables_#5}
  ]
  particleUpdateScript [
   &ParticleState()
   &KillParticles(Kill Particles=&Boolean_NotOperation(Bool=$Particles.IsEndPosition))
   /SetVariables(Particles.HitScanAge:NiagaraFloat=&MakeCustomFloatFromBool(
     Bool=$Particles.IsHitScan
     True Float=$Particles.NormalizedAge
     False Float=&ClampFloat(Float=&Add_Float(A=&Multiply_Float(A=$Particles.NormalizedAge, B=0.5), B=0.66))
    )) {isSetParameters=true, module=SetVariables_#6}
   /SetVariables(Particles.Position=$Particles.Initial.Position) {isSetParameters=true, module=SetVariables_#7}
   /SetVariables(Particles.Position=&MakeCustomVectorFromBool(
     Bool=$Particles.IsEndPosition
     True Vector=&Lerp_Vector(
      A=Linked() {linkedVariable {name=Particles.PreviousPosition, type=NiagaraPosition}}
      B=Linked() {linkedVariable {name=Particles.Position, type=NiagaraPosition}}
      Alpha=&FloatFromCurve(
       FloatCurve=DataInterface() {propertyValues {Curve {keys [{arriveTangent=2.1933, time=0, value=0}, <at 0.66 1>]}}}
       CurveIndex=$Particles.HitScanAge
      )
     )
     False Vector=Linked() {linkedVariable {name=Particles.Position, type=NiagaraPosition}}
    )) {isSetParameters=true}
   /SetVariables(Particles.Tracer Scale:NiagaraFloat=&ClampFloat(Float=&Divide_Float(
      A=&VectorLength(
       Vector=&SubtractVectorFromPosition(Position=$Particles.Position, Vector=Linked() {linkedVariable {name=Particles.OriginPosition, type=NiagaraPosition}})
       Safe Length Calculation=false
      )
      B=$Particles.ProjectileLength
     ))) {isSetParameters=true, module=SetVariables_#8}
   &ScaleSpriteSize(
    Scale Factor=&MakeVector2D(
     X=1
     Y=&FloatFromCurve(
      FloatCurve=DataInterface() {propertyValues {Curve {keys [<at 0 0>, <at 0.33 1>, {arriveTangent=-4.3415, time=0.66, value=1}, {arriveTangent=-1.2985, leaveTangent=-1.2985, time=1, value=0}]}}}
      CurveIndex=$Particles.HitScanAge
      Scale Curve=$Particles.Tracer Scale
     )
    )
    Scale Sprite Size Mode=::NewEnumerator2
   ) {module=ScaleSpriteSize001}
   /SetVariables(Particles.Muzzle Position Offset:Vector3f=&Multiply_VectorByFloat(
     Vector=$Particles.Direction
     Float=&Multiply_Float(A=&MakeFloatFromVector2d(Vector 2D=$Particles.SpriteSize, Channel=::NewEnumerator1), B=-0.49)
    )) {isSetParameters=true, module=SetVariables_#9}
   /SetVariables(Particles.Position=&AddVectorToPosition(Position=$Particles.Position, Vector=$Particles.Muzzle Position Offset)) {isSetParameters=true, module=SetVariables_#10}
   /SetVariables(Particles.Color=&self:Material_Hue_Shift_01(Color:LinearColor=<rgba 2 0.5 0.05 1>, Hue Shift:NiagaraFloat=$Particles.HueShift)) {isSetParameters=true, module=SetVariables_#11}
  ]
  rendererClasses [&NiagaraSpriteRendererProperties]
  renderers [{rendererClass=&NiagaraSpriteRendererProperties}]
 }
 Local_Player_Trail {
  emitterUpdateScript [
   &EmitterState()
   &Spawn_NDCFortProjectiles(Instigator Filter=1, Reader:NiagaraDataInterfaceDataChannelRead=$System.NDCReader)
  ]
  particleSpawnScript [
   &InitializeParticleFromNDC_FortProjectiles(
    FallbackAmmoType=2
    LOD Output Scale Max=2
    Lifetime=0.75
    Scale Array=DataInterface() {propertyValues {GpuSyncMode=SyncCpuToGpu, MaxElements=4, FloatData [15,20,24,14]}}
    Scale by Array=true
    Use LOD Distance Scaling=true
   )
   &InitializeParticle(
    Color=&self:Material_Hue_Shift_02(Color:LinearColor=<rgba 2 1.2018 0 1>, Hue Shift:NiagaraFloat=$Particles.HueShift)
    Lifetime=1.35
    Mesh Scale Mode=::NewEnumerator3
    Mesh Uniform Scale=10
    Position=$Particles.CurrentPosition
    Position Mode=::NewEnumerator1
    Ribbon Width=&Multiply_Float(A=$Particles.ProjectileScale, B=&Lerp_Float(A=0.75, B=1, Alpha=$Particles.Random))
    Ribbon Width Mode=::NewEnumerator1
    UsePositionOffset=false
    Write Material Random=true
   )
   /SetVariables(Particles.Smoke WPO Scale:NiagaraFloat=&SelectFloatFromArray(Direct Array Index=$Particles.AmmoType, Float Selection Array=DataInterface() {propertyValues {FloatData [0.75,1,1.2,0.75]}})) {isSetParameters=true, module=SetVariables_#12}
   /SetVariables(Particles.IsHitScan:NiagaraBool=&Boolean_AndOperation(
     Bool1=$Particles.IsEndPosition
     Bool2=&SetBoolByFloatComparison(
      A=&VectorLength(
       Vector=&SubtractVectorFromPosition(Position=$Particles.PreviousPosition, Vector=Linked() {linkedVariable {name=Particles.OriginPosition, type=NiagaraPosition}})
       Safe Length Calculation=false
      )
      B=0
      Comparison Type=::NewEnumerator4
     )
    )) {isSetParameters=true}
  ]
  particleUpdateScript [
   &ParticleState()
   &ScaleColor(
    Scale Alpha=&FloatFromCurve(
     FloatCurve=DataInterface() {propertyValues {Curve {keys [
         {interpMode=RCIM_Cubic, leaveTangentWeight=0.3333, tangentMode=RCTM_User, time=0, value=1}
         {arriveTangent=-3.46, interpMode=RCIM_Cubic, leaveTangent=-3.46, tangentMode=RCTM_User, time=1, value=0}
        ]}}}
     Scale Curve=&FloatFromCurve(FloatCurve=DataInterface() {propertyValues {Curve {keys [<at 0 0.25>,<at 0.05 1>]}}}, CurveIndex=$Particles.RibbonUVDistance)
    )
    Scale RGB=&Multiply_VectorByFloat(Vector=<1 1 1>, Float=1)
    ScaleA=true
    ScaleRGB=true
   )
   &ScaleRibbonWidth(Ribbon Width Scale=&FloatFromCurve(FloatCurve=DataInterface() {propertyValues {Curve {keys [{arriveTangent=1.4563, interpMode=RCIM_Cubic, leaveTangent=1.4563, tangentMode=RCTM_User, time=0, value=0.4}, <at 1 1 RCIM_Cubic>]}}}))
   &DynamicMaterialParameters(
    Index 0 Param 1=&FloatFromCurve(
     FloatCurve=DataInterface() {propertyValues {Curve {keys [
         {arriveTangent=1.4179, interpMode=RCIM_Cubic, leaveTangent=1.4179, tangentMode=RCTM_User, time=0.25, value=0}
         {interpMode=RCIM_Cubic, tangentMode=RCTM_User, time=1, value=1}
        ]}}}
     Scale Curve=&Multiply_Float(A=8, B=$Particles.Smoke WPO Scale)
    )
    Index 0 Param 2=&FloatFromCurve(
     FloatCurve=DataInterface() {propertyValues {Curve {keys [
         {arriveTangent=4.1785, interpMode=RCIM_Cubic, leaveTangent=4.1785, tangentMode=RCTM_User, time=0, value=0}
         {arriveTangent=1.1179, interpMode=RCIM_Cubic, leaveTangent=1.1179, tangentMode=RCTM_User, time=0.2042, value=0.6228}
         {arriveTangent=0.0224, interpMode=RCIM_Cubic, leaveTangent=0.0224, tangentMode=RCTM_User, time=1, value=1}
        ]}}}
     Scale Curve=2.5
    )
    Index 0 Param 3=$Particles.RandomFloat
    Index 0 Param 4=&FloatFromCurve(
     FloatCurve=DataInterface() {propertyValues {Curve {keys [{interpMode=RCIM_Cubic, tangentMode=RCTM_User, time=0, value=0}, {interpMode=RCIM_Cubic, tangentMode=RCTM_User, time=0.33, value=1}]}}}
     Scale Curve=0.8
    )
    Module.Index0Param0Write=true
    Module.Index0Param1Write=true
    Module.Index0Param2Write=true
    Module.Index0Param3Write=true
   )
   /SetVariables(Particles.HitScanAge:NiagaraFloat=&MakeCustomFloatFromBool(
     Bool=$Particles.IsHitScan
     True Float=$Particles.NormalizedAge
     False Float=&ClampFloat(Float=&Add_Float(A=&Multiply_Float(A=$Particles.NormalizedAge, B=0.5), B=0.66))
    )) {isSetParameters=true}
   /SetVariables(Particles.Position=$Particles.Initial.Position) {isSetParameters=true}
   /SetVariables(Particles.Position=&MakeCustomVectorFromBool(
     Bool=$Particles.IsEndPosition
     True Vector=&Lerp_Vector(
      A=Linked() {linkedVariable {name=Particles.PreviousPosition, type=NiagaraPosition}}
      B=Linked() {linkedVariable {name=Particles.Position, type=NiagaraPosition}}
      Alpha=&FloatFromCurve(
       FloatCurve=DataInterface() {propertyValues {Curve {keys [{arriveTangent=2.1933, time=0, value=0}, <at 0.2 1>]}}}
       CurveIndex=$Particles.HitScanAge
      )
     )
     False Vector=Linked() {linkedVariable {name=Particles.Position, type=NiagaraPosition}}
    )) {isSetParameters=true}
  ]
  rendererClasses [&NiagaraRibbonRendererProperties]
  renderers [{rendererClass=&NiagaraRibbonRendererProperties}]
 }
 Sparks {
  emitterUpdateScript [
   &EmitterState()
   &Spawn_NDCFortProjectiles(
    Instigator Filter=1
    Reader:NiagaraDataInterfaceDataChannelRead=$System.NDCReader
    Spawn Count:NiagaraInt32=&RandomRangeInt(Minimum=14, Maximum=20)
   )
  ]
  particleSpawnScript [
   &InitializeParticleFromNDC_FortProjectiles(
    FallbackAmmoType=2
    FallbackOrigin:NiagaraPosition=<0 50 0>
    LOD Output Scale Min=0.5
    Length Scale Array=DataInterface() {propertyValues {GpuSyncMode=SyncCpuToGpu, MaxElements=4, FloatData [0.75,0.85,1,0.7]}}
    Length Scale by Array=true
    Length by Velocity=true
    Scale Array=DataInterface() {propertyValues {GpuSyncMode=SyncCpuToGpu, MaxElements=4, FloatData [8,12,16,6]}}
    Scale by Array=true
    Use LOD Distance Scaling=true
   )
   &self:GetADSVariable_01() {enabled=false, module=GetADSVariable_01}
   &KillParticles(Kill Particles=$Particles.IsStartPosition) {module=KillParticles001}
   &InitializeParticle(
    Color Mode=::NewEnumerator0
    Lifetime Max=0.6
    Lifetime Min=0.2
    Lifetime Mode=::NewEnumerator1
    Position=$Particles.CurrentPosition
    Position Mode=::NewEnumerator1
    Sprite Size Mode=::NewEnumerator3
    Uniform Sprite Size=&RandomRangeFloat(Minimum=10, Maximum=20)
    UsePositionOffset=false
    Write Material Random=true
   )
   &DynamicMaterialParameters(
    Index 0 Param 1=2
    Index 0 Param 2=1
    Index 0 Param 3=3
    Index 0 Param 4=0.25
    Module.Index0Param0Write=true
    Module.Index0Param1Write=true
    Module.Index0Param2Write=true
    Module.Index0Param3Write=true
   )
   /SetVariables(Particles.Color=&MakeCustomLinearColorFromBool(Bool=$Particles.IsEndPosition, True Color=<rgba 2 0.5 0.05 0>, False Color=<rgba 2 0.5 0.05 1>)) {isSetParameters=true, module=SetVariables_#4}
   /SetVariables(Particles.IsHitScan:NiagaraBool=&Boolean_AndOperation(
     Bool1=$Particles.IsEndPosition
     Bool2=&SetBoolByFloatComparison(
      A=&VectorLength(
       Vector=&SubtractVectorFromPosition(Position=$Particles.PreviousPosition, Vector=Linked() {linkedVariable {name=Particles.OriginPosition, type=NiagaraPosition}})
       Safe Length Calculation=false
      )
      B=0
      Comparison Type=::NewEnumerator4
     )
    )) {isSetParameters=true, module=SetVariables_#5}
  ]
  particleUpdateScript [
   /SetVariables(Transient.OldPosition:NiagaraPosition=$Particles.Position) {isSetParameters=true}
   &ParticleState()
   &ShapeLocation(MaskSpawnGroup=false, Override Seed=false, UseOwnerScale=true)
   &KillParticles(Kill Particles=&Boolean_NotOperation(Bool=$Particles.IsEndPosition))
   /SetVariables(Particles.HitScanAge:NiagaraFloat=&MakeCustomFloatFromBool(
     Bool=$Particles.IsHitScan
     True Float=$Particles.NormalizedAge
     False Float=&ClampFloat(Float=&Add_Float(A=&Multiply_Float(A=$Particles.NormalizedAge, B=0.5), B=0.66))
    )) {isSetParameters=true, module=SetVariables_#6}
   /SetVariables(Particles.Position=$Particles.Initial.Position) {isSetParameters=true, module=SetVariables_#7}
   /SetVariables(Particles.Position=&Lerp_Vector(
     A=Linked() {linkedVariable {name=Particles.PreviousPosition, type=NiagaraPosition}}
     B=Linked() {linkedVariable {name=Particles.Position, type=NiagaraPosition}}
     Alpha=&RandomRangeFloat(Minimum=0, Maximum=1)
    )) {isSetParameters=true}
   /SetVariables(Particles.Tracer Scale:NiagaraFloat=&ClampFloat(Float=&Divide_Float(
      A=&VectorLength(
       Vector=&SubtractVectorFromPosition(Position=$Particles.Position, Vector=Linked() {linkedVariable {name=Particles.OriginPosition, type=NiagaraPosition}})
       Safe Length Calculation=false
      )
      B=$Particles.ProjectileLength
     ))) {isSetParameters=true, module=SetVariables_#8}
   &ScaleSpriteSize(
    Scale Factor=&MakeVector2D(
     X=1
     Y=&FloatFromCurve(
      FloatCurve=DataInterface() {propertyValues {Curve {keys [<at 0 0>, <at 0.33 1>, {arriveTangent=-4.3415, time=0.66, value=1}, {arriveTangent=-1.2985, leaveTangent=-1.2985, time=1, value=0}]}}}
      CurveIndex=$Particles.HitScanAge
      Scale Curve=$Particles.Tracer Scale
     )
    )
    Scale Sprite Size Mode=::NewEnumerator2
   ) {module=ScaleSpriteSize001}
   /SetVariables(Particles.Muzzle Position Offset:Vector3f=&Multiply_VectorByFloat(
     Vector=$Particles.Direction
     Float=&Multiply_Float(A=&MakeFloatFromVector2d(Vector 2D=$Particles.SpriteSize, Channel=::NewEnumerator1), B=-0.49)
    )) {isSetParameters=true, module=SetVariables_#9}
   /SetVariables(Particles.Position=&AddVectorToPosition(Position=$Particles.Position, Vector=$Particles.Muzzle Position Offset)) {isSetParameters=true, module=SetVariables_#10}
   /SetVariables(Particles.Position=&AddVectorToPosition(Position=$Particles.Position, Vector=<0 0 0>)) {isSetParameters=true}
   /SetVariables(Particles.Color=<rgba 2 1.2966 0 1>) {isSetParameters=true, module=SetVariables_#11}
   &FN_Light_Attributes(
    Color=$StackContext.Color
    Create Color Attribute=true
    Create Exponent Attribute=true
    Create Radius Attribute=true
    Exponent=2
    Radius=300
    Scale Color=3.3
   )
   /Niagara/Modules/Update/Utility/VectorNoiseValue(
    Noise Frequency=5
    Pan Noise=true
    Pan Noise Field:Vector3f=<0 0 0.1>
    Sample Position:NiagaraPosition=$StackContext.ShapeLocation.ShapePosition
   )
   /SetVariables(Particles.Position=&AddVectorToPosition(Position=$Particles.Position, Vector=&Multiply_VectorByFloat(Vector=$Output.VectorNoiseValue.Noise, Float=10))) {isSetParameters=true}
   /SetVariables(Particles.Position=&AddVectorToPosition(Position=$Particles.Position, Vector=&Multiply_VectorByFloat(Vector=$Particles.ProjectileDirection, Float=1))) {enabled=false, isSetParameters=true}
   &ScaleSpriteSize(
    Uniform Curve Scale=&Lerp_Float(A=1.2, B=0.6, Alpha=&MakeFloatFromInt(Int=$Particles.CustomData))
    Uniform Curve Sprite Scale=DataInterface() {propertyValues {Curve {
       defaultValue=0
       keys [
        <at 0 0 RCIM_Cubic>
        {arriveTangentWeight=0.0556, interpMode=RCIM_Cubic, leaveTangentWeight=0.3611, tangentMode=RCTM_User, time=0.2686, value=0.9922}
        {interpMode=RCIM_Cubic, tangentMode=RCTM_User, time=0.75, value=1}
        {arriveTangent=-7.3785, interpMode=RCIM_Cubic, leaveTangent=-7.3785, tangentMode=RCTM_User, time=1, value=0}
       ]
      }}}
   )
   /SetVariables() {isSetParameters=true}
   &ScaleSpriteSize(
    Scale Factor=&Lerp_Vector2(
     A=<1 1>
     B=<0.7 2>
     Alpha=&RemapRange(
      Input Value To Remap=&VectorLength(Vector=$StackContext.CalculatedVelocity)
      Input Min=0
      Input Max=100
      Output Min=0
      Output Max=1
     )
    )
    Scale Sprite Size Mode=::NewEnumerator2
   ) {module=ScaleSpriteSize002}
   &FN_Sprite_Attributes(
    Color=&MultiplyLinearColor(
     Linear Color=$StackContext.Color
     Scale Factor=&MakeVector4FromVectorAndFloat("Vector (RGB)"=<1 1 1>, "Float (Alpha)"=0.025)
    )
    Create Color Attribute=true
    Create SpriteSize Attribute=true
    SpriteSize=&Multiply_Vector2D(A=$StackContext.SpriteSize, B=&Vector2DFromFloat(Value=5))
    Write to Module Namespace=true
   )
  ]
  rendererClasses [&NiagaraSpriteRendererProperties]
  renderers [
   {rendererClass=&NiagaraSpriteRendererProperties}
   {rendererClass=&NiagaraSpriteRendererProperties, rendererIndex=1}
  ]
 }
 Trail {
  emitterUpdateScript [
   &EmitterState()
   &Spawn_NDCFortProjectiles(Instigator Filter=2, Reader:NiagaraDataInterfaceDataChannelRead=$System.NDCReader)
  ]
  particleSpawnScript [
   &InitializeParticleFromNDC_FortProjectiles(
    FallbackAmmoType=2
    LOD Output Scale Max=2
    Scale Array=DataInterface() {propertyValues {GpuSyncMode=SyncCpuToGpu, MaxElements=4, FloatData [15,20,24,14]}}
    Scale by Array=true
    Use LOD Distance Scaling=true
   )
   &InitializeParticle(
    Color=&self:Material_Hue_Shift(
     Color:LinearColor=&MakeCustomLinearColorFromBool(Bool=$Particles.IsSuppressed, True Color=<rgba 0.15 0.0225 0 5>, False Color=<rgba 0.5 0.075 0 5>)
     Hue Shift:NiagaraFloat=$Particles.HueShift
    )
    Lifetime=&Multiply_Float(
     A=1
     B=&SelectFloatFromArray(Direct Array Index=$Particles.AmmoType, Float Selection Array=DataInterface() {propertyValues {FloatData [0.75,1,2,0.7]}})
    )
    Mesh Scale Mode=::NewEnumerator3
    Mesh Uniform Scale=10
    Position=$Particles.CurrentPosition
    Position Mode=::NewEnumerator1
    Ribbon Width=&Multiply_Float(A=$Particles.ProjectileScale, B=&Lerp_Float(A=0.75, B=1, Alpha=$Particles.Random))
    Ribbon Width Mode=::NewEnumerator1
    UsePositionOffset=false
    Write Material Random=true
   )
   /SetVariables(Particles.Smoke WPO Scale:NiagaraFloat=&SelectFloatFromArray(Direct Array Index=$Particles.AmmoType, Float Selection Array=DataInterface() {propertyValues {FloatData [0.75,1,1.2,0.75]}})) {isSetParameters=true, module=SetVariables_#12}
   /SetVariables(Particles.IsHitScan:NiagaraBool=&Boolean_AndOperation(
     Bool1=$Particles.IsEndPosition
     Bool2=&SetBoolByFloatComparison(
      A=&VectorLength(
       Vector=&SubtractVectorFromPosition(Position=$Particles.PreviousPosition, Vector=Linked() {linkedVariable {name=Particles.OriginPosition, type=NiagaraPosition}})
       Safe Length Calculation=false
      )
      B=0
      Comparison Type=::NewEnumerator4
     )
    )) {isSetParameters=true}
  ]
  particleUpdateScript [
   &ParticleState()
   &ScaleColor(
    Scale Alpha=&FloatFromCurve(
     FloatCurve=DataInterface() {propertyValues {Curve {keys [
         {arriveTangent=-2.5785, interpMode=RCIM_Cubic, time=0.5, value=1}
         {arriveTangent=-3.9991, arriveTangentWeight=1.0602, interpMode=RCIM_Cubic, leaveTangent=-3.9991, tangentMode=RCTM_User, time=1, value=0}
        ]}}}
     Scale Curve=0.2
    )
    Scale RGB=&Multiply_VectorByFloat(
     Vector=<1 1 1>
     Float=&FloatFromCurve(
      FloatCurve=DataInterface() {propertyValues {Curve {keys [
          {arriveTangent=-13.4341, interpMode=RCIM_Cubic, time=0, value=1}
          {arriveTangent=-9.3354, interpMode=RCIM_Cubic, leaveTangent=-9.3354, tangentMode=RCTM_User, time=0.2, value=0}
         ]}}}
      CurveIndex=$Particles.Age
     )
    )
    ScaleA=true
    ScaleRGB=true
   )
   &ScaleRibbonWidth(Ribbon Width Scale=&FloatFromCurve(FloatCurve=DataInterface() {propertyValues {Curve {keys [{arriveTangent=1.9828, interpMode=RCIM_Cubic, leaveTangent=1.9828, tangentMode=RCTM_User, time=0, value=0.4}, <at 1 1.5 RCIM_Cubic>]}}}))
   &DynamicMaterialParameters(
    Index 0 Param 1=&FloatFromCurve(
     FloatCurve=DataInterface() {propertyValues {Curve {keys [
         {arriveTangent=1.4179, interpMode=RCIM_Cubic, leaveTangent=1.4179, tangentMode=RCTM_User, time=0.25, value=0}
         {interpMode=RCIM_Cubic, tangentMode=RCTM_User, time=1, value=1}
        ]}}}
     Scale Curve=&Multiply_Float(A=32, B=$Particles.Smoke WPO Scale)
    )
    Index 0 Param 2=&FloatFromCurve(
     FloatCurve=DataInterface() {propertyValues {Curve {keys [
         {arriveTangent=4.1785, interpMode=RCIM_Cubic, leaveTangent=4.1785, tangentMode=RCTM_User, time=0, value=0}
         {arriveTangent=1.1179, interpMode=RCIM_Cubic, leaveTangent=1.1179, tangentMode=RCTM_User, time=0.2042, value=0.6228}
         {arriveTangent=0.0224, interpMode=RCIM_Cubic, leaveTangent=0.0224, tangentMode=RCTM_User, time=1, value=1}
        ]}}}
     Scale Curve=-2.5
    )
    Index 0 Param 3=$Particles.RandomFloat
    Index 0 Param 4=&FloatFromCurve(
     FloatCurve=DataInterface() {propertyValues {Curve {keys [{interpMode=RCIM_Cubic, tangentMode=RCTM_User, time=0, value=0}, {interpMode=RCIM_Cubic, tangentMode=RCTM_User, time=0.33, value=1}]}}}
     Scale Curve=0.8
    )
    Module.Index0Param0Write=true
    Module.Index0Param1Write=true
    Module.Index0Param2Write=true
    Module.Index0Param3Write=true
   )
   /SetVariables(Particles.HitScanAge:NiagaraFloat=&MakeCustomFloatFromBool(
     Bool=$Particles.IsHitScan
     True Float=$Particles.NormalizedAge
     False Float=&ClampFloat(Float=&Add_Float(A=&Multiply_Float(A=$Particles.NormalizedAge, B=0.5), B=0.66))
    )) {isSetParameters=true}
   /SetVariables(Particles.Position=$Particles.Initial.Position) {isSetParameters=true}
   /SetVariables(Particles.Position=&MakeCustomVectorFromBool(
     Bool=$Particles.IsEndPosition
     True Vector=&Lerp_Vector(
      A=Linked() {linkedVariable {name=Particles.PreviousPosition, type=NiagaraPosition}}
      B=Linked() {linkedVariable {name=Particles.Position, type=NiagaraPosition}}
      Alpha=&FloatFromCurve(
       FloatCurve=DataInterface() {propertyValues {Curve {keys [{arriveTangent=2.1933, time=0, value=0}, <at 0.2 1>]}}}
       CurveIndex=$Particles.HitScanAge
      )
     )
     False Vector=Linked() {linkedVariable {name=Particles.Position, type=NiagaraPosition}}
    )) {isSetParameters=true}
  ]
  rendererClasses [&NiagaraRibbonRendererProperties]
  renderers [{rendererClass=&NiagaraRibbonRendererProperties}]
 }
}
summary {
 systemName=NS_NDC_ProjectilesHandler_GoldenGun
 emitters [
  {emitterName=Trail, rendererClasses [&NiagaraRibbonRendererProperties]}
  {emitterName=Bullet, rendererClasses [&NiagaraSpriteRendererProperties]}
  {emitterName=Local_Player_Trail, rendererClasses [&NiagaraRibbonRendererProperties]}
  {emitterName=Sparks, rendererClasses [&NiagaraSpriteRendererProperties]}
 ]
 userVariables [User.IsADS {type=NiagaraFloat, defaultValue=0, description=""}]
}
